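# Sound cue for the V-Bucks menu hover sound in the front-end UI, variant 03 by its name.
# A single wave player feeds the Fortnite_UI_VBucks_Menu_Hover_03 sound wave straight to the output, with no randomization, modulation or mixing nodes.
# The cue is assigned to the Fort_UI sound class and carries a very large max distance.
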SoundCue /Game/Sounds/Fort_UI_Sounds/FrontEnd/New_FrontEnd/Fortnite_UI_VBucks_Menu_Hover_03_Cue
root=WavePlayer_0
nodes [{id=WavePlayer_0, properties {soundWaveAssetPtr=&/Fortnite_UI_VBucks_Menu_Hover_03}}]
properties {maxDistance=2097152, soundClassObject=&Fort_UI}
