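# Battle Pass gift received screen, a FortMessageReceivedScreen widget showing header, body and Battle Pass icon over a dimmed, blurred background.
# On Construct it switches to the default front-end camera, plays the Intro scale/fade animation, then sets text from the gift box definition's defaults.
# Accept deactivates the widget; deactivation calls NavigateForGift and broadcasts On Message Closed.
# Nodes for header/body override-string params exist but are not wired in.
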WidgetBlueprint /Game/UI/ItemReceived/BattlePassGiftReceived
animations [Intro {
  binding_count=3
  display_label=""
  length_seconds=1.3001
  object=&self:Intro
  sequence {bindings [
    ShowNextButton {
     id=#1
     possessed_class=IconTextButton_C
     tracks [
      {
       type=MovieScene2DTransformTrack
       sections [{
         type=MovieScene2DTransformSection
         end_frame=18
         start_frame=9
         channels [
          Translation.X {type=MovieSceneScriptingFloatChannel}
          Translation.Y {type=MovieSceneScriptingFloatChannel, thinned_from=4, keys [[9,50],[18.0007,0],[18.001,0]]}
          Angle {type=MovieSceneScriptingFloatChannel}
          Scale.X {type=MovieSceneScriptingFloatChannel}
          Scale.Y {type=MovieSceneScriptingFloatChannel}
          Shear.X {type=MovieSceneScriptingFloatChannel}
          Shear.Y {type=MovieSceneScriptingFloatChannel}
         ]
        }]
      }
      {
       type=MovieSceneFloatTrack
       display_name=Render Opacity
       sections [{
         type=MovieSceneFloatSection
         end_frame=26
         start_frame=0
         channels [None {type=MovieSceneScriptingFloatChannel, keys [[0,0],[11.001,0],[26.001,1]]}]
        }]
      }
     ]
    }
    HeaderContainer {
     id=#2
     display_name=VerticalBox_6
     possessed_class=HorizontalBox
     tracks [{
       type=MovieScene2DTransformTrack
       sections [{
         type=MovieScene2DTransformSection
         end_frame=16
         start_frame=0
         channels [
          Translation.X {type=MovieSceneScriptingFloatChannel}
          Translation.Y {type=MovieSceneScriptingFloatChannel}
          Angle {type=MovieSceneScriptingFloatChannel}
          Scale.X {type=MovieSceneScriptingFloatChannel, keys [[6.001,0],[11.001,1.2],[13.001,1]]}
          Scale.Y {type=MovieSceneScriptingFloatChannel, keys [[6.001,0],[11.001,1.2],[13.001,1]]}
          Shear.X {type=MovieSceneScriptingFloatChannel}
          Shear.Y {type=MovieSceneScriptingFloatChannel}
         ]
        }]
      }]
    }
    Overlay_0 {
     id=#3
     possessed_class=Overlay
     tracks [{
       type=MovieScene2DTransformTrack
       sections [{
         type=MovieScene2DTransformSection
         end_frame=11
         start_frame=0
         channels [
          Translation.X {type=MovieSceneScriptingFloatChannel}
          Translation.Y {type=MovieSceneScriptingFloatChannel}
          Angle {type=MovieSceneScriptingFloatChannel}
          Scale.X {type=MovieSceneScriptingFloatChannel, keys [[0,0],[3.999,1.3],[7,1]]}
          Scale.Y {type=MovieSceneScriptingFloatChannel, keys [[0,0],[3.999,1.3],[7.0003,1]]}
          Shear.X {type=MovieSceneScriptingFloatChannel}
          Shear.Y {type=MovieSceneScriptingFloatChannel}
         ]
        }]
      }]
    }
   ]}
 }]
graphs [EventGraph {nodes [
   "On Button Clicked (ShowNextButton)"!() -> (OutputDelegate, then, Button:Common Button Legacy Object Reference) {is=title K2Node_ComponentBoundEvent}
   > DeactivateWidget()
   Event Construct()
   > SetFrontEndCamera!(
    execute
    self:Fort Front End Context Object Reference=Get FortFrontEndContext!() -> (ReturnValue:Fort Front End Context Object Reference) {is=title K2Node_GetBlueprintContextEx}
    NewState:EFrontEndCamera Enum=FrontendDefault
    TransitionParams:View Target Transition Params Structure=(BlendTime=0,BlendFunction=VTBlend_Cubic,BlendExp=2,bLockOutgoing=False)
   ) -> (then)
   > PlayAnimation(InAnimation=Get Intro())
   > Delay(Duration=0)
   n1 = Get GiftBoxItem() -> (GiftBoxItem:Fort Gift Box Item Object Reference)
   > "SetText (Text)"(
    self=Get Header() -> (Header:Common Text Object Reference)
    InText=Get DefaultHeaderText!(self:Fort Gift Box Item Definition Object Reference=GetGiftBoxDefinition!(self:Fort Gift Box Item Object Reference=@n1) -> (ReturnValue:Fort Gift Box Item Definition Object Reference)) -> (DefaultHeaderText:Text)
   )
   n2 = Get GiftBoxItem()
   > "SetText (Text)"(
    self=Get Body() -> (Body:Common Text Object Reference)
    InText=Get DefaultBodyText!(self:Fort Gift Box Item Definition Object Reference=GetGiftBoxDefinition(self=@n2)) -> (DefaultBodyText:Text)
   )
   Event On Deactivated()
   > NavigateForGift!(execute, Item:Fort Gift Box Item Object Reference=Get GiftBoxItem()) -> (then)
   > Call On Message Closed!(execute:Exec, self:Battle Pass Gift Received Object Reference) -> (then:Exec) {is=title K2Node_CallDelegate}
   FormatTextUsingParams!(self:Fort Gift Box Item Object Reference=@n2, FormatString:"Text (by ref)"=MakeLiteralText(Value="LOCTEXT(\"{BodyOverrideString}\")")) -> (ReturnValue:Text)
   FormatTextUsingParams(self=@n1, FormatString=MakeLiteralText(Value="LOCTEXT(\"{HeaderOverrideString}\")"))
   HasParam!(self:Fort Gift Box Item Object Reference=@n1, ParamName:String=HeaderOverrideString) -> (ReturnValue:bool)
   HasParam(self=@n2, ParamName=BodyOverrideString)
   Event PreConstruct()
   Event Tick()
  ]}]
variables []
widgets {
 info {inheritedWidgetCount=1, parentClass=/Script/FortniteUI.FortMessageReceivedScreen, rootWidgetClass=&Overlay, widgetCount=26}
 widgets [
  Overlay_2 {widgetClassPath=&Overlay, widget=Overlay_2}
  Border_0 {
   widgetClassPath=&Border
   parent=Overlay_2
   widget=Border_0
   properties {background {
     bIsDynamicallyLoaded=false
     drawAs=Image
     imageType=NoImage
     mirroring=NoMirror
     resourceName=None
     resourceObject=null
     tiling=NoTile
     imageSize=<32 32>
     margin=<ltrb 0 0 0 0>
     outlineSettings {
      bUseBrushTransparency=false
      roundingType=HalfHeightRadius
      width=0
      color {colorUseRule=UseColor_Specified, specifiedColor=<rgba 0 0 0 0>}
      cornerRadii=<0 0 0 0>
     }
     tintColor {colorUseRule=UseColor_Specified, specifiedColor=<rgba 0 0 0 0.75>}
     uVRegion {bIsValid=false, max=<0 0>, min=<0 0>}
    }}
   slot {class=OverlaySlot, properties {horizontalAlignment=HAlign_Fill, verticalAlignment=VAlign_Fill}}
  }
  BackgroundBlur_0 {
   widgetClassPath=&BackgroundBlur
   parent=Overlay_2
   widget=BackgroundBlur_0
   properties {blurStrength=25}
   slot {class=OverlaySlot, properties {horizontalAlignment=HAlign_Fill, verticalAlignment=VAlign_Fill}}
  }
  MainSafeZone {
   widgetClassPath=&SafeZone
   parent=Overlay_2
   widget=MainSafeZone
   slot {class=OverlaySlot, properties {horizontalAlignment=HAlign_Fill, verticalAlignment=VAlign_Fill}}
  }
  HorizontalBox_2 {
   widgetClassPath=&HorizontalBox
   parent=MainSafeZone
   widget=HorizontalBox_2
   slot {class=SafeZoneSlot, properties {hAlign=HAlign_Center}}
  }
  SizeBox_0 {
   widgetClassPath=&SizeBox
   parent=HorizontalBox_2
   widget=SizeBox_0
   properties {bOverride_WidthOverride=true, maxDesiredWidth=2568.8411, widthOverride=1384.8259}
   slot {class=HorizontalBoxSlot, properties {horizontalAlignment=HAlign_Left, padding=<ltrb 50 90 0 80>}}
  }
  MainContainer {widgetClassPath=&VerticalBox, parent=SizeBox_0, widget=MainContainer, slot {class=SizeBoxSlot}}
  VerticalBox_6 {
   widgetClassPath=&VerticalBox
   parent=MainContainer
   widget=VerticalBox_6
   slot {class=VerticalBoxSlot, properties {horizontalAlignment=HAlign_Center, verticalAlignment=VAlign_Center}}
  }
  VerticalBox_7 {
   widgetClassPath=&VerticalBox
   parent=VerticalBox_6
   widget=VerticalBox_7
   slot {class=VerticalBoxSlot, properties {horizontalAlignment=HAlign_Center}}
  }
  Header {
   widgetClassPath=&CommonTextBlock
   bIsVariable=true
   parent=VerticalBox_6
   widget=Header
   properties {
    justification=Center
    scrollStyle=&TextScrollStyle-Base_C
    style=/FortUILibrary/Styles/Typography/Foundation/TextStyle-Header-72-Horizon-SeasonPass.TextStyle-Header-72-Horizon-SeasonPass_C
    text="BATTLE PASS HEADER![ph]"
    font {
     bForceMonospaced=false
     bMaterialIsStencil=false
     fontMaterial=/FortUILibrary/Materials/Foundation/M_Font_Horizon_SeasonPass
     fontObject=&Burbank
     letterSpacing=0
     monospacedWidth=1
     size=72
     skewAmount=0
     typefaceFontName=Burbank-Black
     outlineSettings {
      bApplyOutlineToDropShadows=true
      bMiteredCorners=false
      bSeparateFillAlpha=false
      outlineMaterial=null
      outlineSize=3
      outlineColor=<rgba 0 0 0 0.07>
     }
    }
    shadowOffset=<0 0>
   }
   slot {class=VerticalBoxSlot, properties {horizontalAlignment=HAlign_Center, verticalAlignment=VAlign_Top, padding=<ltrb 0 -10 0 -20>}}
  }
  Body {
   widgetClassPath=&CommonTextBlock
   bIsVariable=true
   parent=VerticalBox_6
   widget=Body
   properties {
    autoWrapText=true
    justification=Center
    style=/FortUILibrary/Styles/Typography/Foundation/TextStyle-Base-M.TextStyle-Base-M_C
    text="battlepass body battlepass body battlepass body battlepass body battlepass body battlepass body battlepass body [ph]"
    wrapTextAt=850
    font {
     bForceMonospaced=false
     bMaterialIsStencil=false
     fontMaterial=null
     fontObject=&NotoSans
     letterSpacing=0
     monospacedWidth=1
     size=24
     skewAmount=0
     typefaceFontName=NotoSans-Regular
     outlineSettings {
      bApplyOutlineToDropShadows=false
      bMiteredCorners=false
      bSeparateFillAlpha=false
      outlineMaterial=null
      outlineSize=0
      outlineColor=<rgba 0 0 0 1>
     }
    }
    shadowOffset=<0 0>
   }
   slot {class=VerticalBoxSlot, properties {horizontalAlignment=HAlign_Center}}
  }
  SizeBox_1 {
   widgetClassPath=&SizeBox
   parent=MainContainer
   widget=SizeBox_1
   properties {bOverride_WidthOverride=true, clipping=ClipToBounds, maxDesiredWidth=2375.1282, widthOverride=821.4651}
   slot {class=VerticalBoxSlot, properties {padding=<ltrb 0 -200 0 -200>, size {sizeRule=Fill, value=1}}}
  }
  Overlay_0 {widgetClassPath=&Overlay, parent=SizeBox_1, widget=Overlay_0, slot {class=SizeBoxSlot}}
  BorderRarityFlare {
   widgetClassPath=&Border
   parent=Overlay_0
   widget=BorderRarityFlare
   properties {
    background {
     bIsDynamicallyLoaded=false
     drawAs=NoDrawType
     imageType=NoImage
     mirroring=NoMirror
     resourceName=None
     resourceObject=null
     tiling=NoTile
     imageSize=<32 32>
     margin=<ltrb 0 0 0 0>
     outlineSettings {
      bUseBrushTransparency=false
      roundingType=HalfHeightRadius
      width=0
      color {colorUseRule=UseColor_Specified, specifiedColor=<rgba 0 0 0 0>}
      cornerRadii=<0 0 0 0>
     }
     tintColor {colorUseRule=UseColor_Specified, specifiedColor=<rgba 1 1 1 1>}
     uVRegion {bIsValid=false, max=<0 0>, min=<0 0>}
    }
    padding=<ltrb 0 0 0 0>
   }
   slot {class=OverlaySlot, properties {horizontalAlignment=HAlign_Center, verticalAlignment=VAlign_Center}}
  }
  ImageRarityFlare {
   widgetClassPath=&Image
   bIsVariable=true
   parent=BorderRarityFlare
   widget=ImageRarityFlare
   properties {
    brush {
     bIsDynamicallyLoaded=false
     drawAs=Image
     imageType=NoImage
     mirroring=NoMirror
     resourceName=None
     resourceObject=/Game/UI/Frontend/Rewards/materials/M_UI_DailyReward_Rarity-Flares
     tiling=NoTile
     imageSize=<256 256>
     margin=<ltrb 0 0 0 0>
     outlineSettings {
      bUseBrushTransparency=false
      roundingType=HalfHeightRadius
      width=0
      color {colorUseRule=UseColor_Specified, specifiedColor=<rgba 0 0 0 0>}
      cornerRadii=<0 0 0 0>
     }
     tintColor {colorUseRule=UseColor_Specified, specifiedColor=<rgba 1 1 1 1>}
     uVRegion {bIsValid=false, max=<0 0>, min=<0 0>}
    }
    renderTransform {angle=0, scale=<2.75 1.875>, shear=<0 0>, translation=<0 0>}
   }
   slot {class=BorderSlot, properties {padding=<ltrb 0 0 0 0>}}
  }
  GiftScrollBox {
   widgetClassPath=&ScrollBox
   bIsVariable=true
   parent=Overlay_0
   widget=GiftScrollBox
   properties {
    allowOverscroll=false
    clipping=Inherit
    consumeMouseWheel=Always
    navigation=&self:WidgetTree.GiftScrollBox.WidgetNavigation_0
    navigationScrollPadding=16
    orientation=Orient_Horizontal
    widgetStyle {
     barThickness=9
     bottomShadowBrush {
      bIsDynamicallyLoaded=false
      drawAs=Box
      imageType=FullColor
      mirroring=NoMirror
      resourceName=../../../Engine/Content/Slate/Common/ScrollBoxShadowBottom.png
      resourceObject=null
      tiling=NoTile
      imageSize=<16 8>
      margin=<ltrb 0.5 0 0.5 1>
      outlineSettings {
       bUseBrushTransparency=false
       roundingType=HalfHeightRadius
       width=0
       color {colorUseRule=UseColor_Specified, specifiedColor=<rgba 0 0 0 0>}
       cornerRadii=<0 0 0 0>
      }
      tintColor {colorUseRule=UseColor_Specified, specifiedColor=<rgba 1 1 1 1>}
      uVRegion {bIsValid=false, max=<0 0>, min=<0 0>}
     }
     horizontalScrolledContentPadding=<ltrb 0 0 1 0>
     leftShadowBrush {
      bIsDynamicallyLoaded=false
      drawAs=Box
      imageType=FullColor
      mirroring=NoMirror
      resourceName=../../../Engine/Content/Slate/Common/ScrollBoxShadowLeft.png
      resourceObject=null
      tiling=NoTile
      imageSize=<0 0>
      margin=<ltrb 1 0.5 0 0.5>
      outlineSettings {
       bUseBrushTransparency=false
       roundingType=HalfHeightRadius
       width=0
       color {colorUseRule=UseColor_Specified, specifiedColor=<rgba 0 0 0 0>}
       cornerRadii=<0 0 0 0>
      }
      tintColor {colorUseRule=UseColor_Specified, specifiedColor=<rgba 1 1 1 1>}
      uVRegion {bIsValid=false, max=<0 0>, min=<0 0>}
     }
     rightShadowBrush {
      bIsDynamicallyLoaded=false
      drawAs=Box
      imageType=FullColor
      mirroring=NoMirror
      resourceName=../../../Engine/Content/Slate/Common/ScrollBoxShadowRight.png
      resourceObject=null
      tiling=NoTile
      imageSize=<0 0>
      margin=<ltrb 0 0.5 1 0.5>
      outlineSettings {
       bUseBrushTransparency=false
       roundingType=HalfHeightRadius
       width=0
       color {colorUseRule=UseColor_Specified, specifiedColor=<rgba 0 0 0 0>}
       cornerRadii=<0 0 0 0>
      }
      tintColor {colorUseRule=UseColor_Specified, specifiedColor=<rgba 1 1 1 1>}
      uVRegion {bIsValid=false, max=<0 0>, min=<0 0>}
     }
     topShadowBrush {
      bIsDynamicallyLoaded=false
      drawAs=NoDrawType
      imageType=FullColor
      mirroring=NoMirror
      resourceName=../../../Engine/Content/Slate/Common/ScrollBoxShadowTop.png
      resourceObject=null
      tiling=NoTile
      imageSize=<0 0>
      margin=<ltrb 0.5 1 0.5 0>
      outlineSettings {
       bUseBrushTransparency=false
       roundingType=HalfHeightRadius
       width=0
       color {colorUseRule=UseColor_Specified, specifiedColor=<rgba 0 0 0 0>}
       cornerRadii=<0 0 0 0>
      }
      tintColor {colorUseRule=UseColor_Specified, specifiedColor=<rgba 1 1 1 1>}
      uVRegion {bIsValid=false, max=<0 0>, min=<0 0>}
     }
     verticalScrolledContentPadding=<ltrb 0 0 0 1>
    }
   }
   slot {class=OverlaySlot, properties {horizontalAlignment=HAlign_Center, verticalAlignment=VAlign_Center}}
  }
  SizeBox_4 {
   widgetClassPath=&SizeBox
   parent=Overlay_0
   widget=SizeBox_4
   properties {bOverride_HeightOverride=true, bOverride_WidthOverride=true, heightOverride=380, widthOverride=380}
   slot {class=OverlaySlot, properties {horizontalAlignment=HAlign_Center, verticalAlignment=VAlign_Center}}
  }
  CommonLazyImage_0 {
   widgetClassPath=/Script/CommonUI.CommonLazyImage
   bIsVariable=true
   parent=SizeBox_4
   widget=CommonLazyImage_0
   properties {brush {
     bIsDynamicallyLoaded=false
     drawAs=Image
     imageType=NoImage
     mirroring=NoMirror
     resourceName=None
     resourceObject=/Game/UI/Foundation/Textures/Icons/Items/T-FNBR-BattlePass-L
     tiling=NoTile
     imageSize=<512 512>
     margin=<ltrb 0 0 0 0>
     outlineSettings {
      bUseBrushTransparency=false
      roundingType=HalfHeightRadius
      width=0
      color {colorUseRule=UseColor_Specified, specifiedColor=<rgba 0 0 0 0>}
      cornerRadii=<0 0 0 0>
     }
     tintColor {colorUseRule=UseColor_Specified, specifiedColor=<rgba 1 1 1 1>}
     uVRegion {bIsValid=false, max=<0 0>, min=<0 0>}
    }}
   slot {class=SizeBoxSlot}
  }
  VerticalBox_0 {
   widgetClassPath=&VerticalBox
   parent=MainContainer
   widget=VerticalBox_0
   slot {class=VerticalBoxSlot, properties {horizontalAlignment=HAlign_Center, verticalAlignment=VAlign_Bottom, padding=<ltrb 0 25 0 0>}}
  }
  HorizontalBoxActions {widgetClassPath=&HorizontalBox, parent=VerticalBox_0, widget=HorizontalBoxActions, slot {class=VerticalBoxSlot}}
  ShowNextButton {
   widgetClassPath=&IconTextButton
   bIsVariable=true
   parent=HorizontalBoxActions
   widget=ShowNextButton
   properties {
    bDisplayInputActionWhenNotInteractable=false
    buttonText=Accept
    contentAlignment=HAlign_Center
    minHeight=70
    minWidth=256
    navigation=&self:WidgetTree.ShowNextButton.WidgetNavigation_0
    onButtonBaseClicked=(&self:WidgetTree.ShowNextButton.HandleOnButtonClicked)
    onButtonBaseDoubleClicked=(&self:WidgetTree.ShowNextButton.HandleOnButtonDoubleClicked)
    onButtonBaseHovered=(&self:WidgetTree.ShowNextButton.HandleOnButtonHovered)
    onButtonBaseUnhovered=(&self:WidgetTree.ShowNextButton.HandleOnButtonUnhovered)
    onSelectedChangedBase=(&self:WidgetTree.ShowNextButton.HandleOnSelectedChanged)
    style=/FortUILibrary/Foundation/Buttons/ButtonStyles/TopBar/ButtonStyle-Outline-Green.ButtonStyle-Outline-Green_C
    hoveredSlateSoundOverride {resourceObject=/Game/Sounds/Fort_UI_Sounds/FrontEnd/New_FrontEnd/Fortnite_UI_VBucks_Menu_Hover_03_Cue}
    triggeringInputAction {dataTable=&InputActionDataTable, rowName=Accept}
   }
   slot {class=HorizontalBoxSlot, properties {horizontalAlignment=HAlign_Center, verticalAlignment=VAlign_Center, padding=<ltrb 12 12 12 12>}}
  }
  Overlay_1 {
   widgetClassPath=&Overlay
   parent=HorizontalBox_2
   widget=Overlay_1
   slot {class=HorizontalBoxSlot, properties {horizontalAlignment=HAlign_Center, verticalAlignment=VAlign_Center, padding=<ltrb 0 0 30 30>}}
  }
  SizeBox_5 {
   widgetClassPath=&SizeBox
   parent=Overlay_1
   widget=SizeBox_5
   properties {heightOverride=350, widthOverride=2800.9375, renderTransformPivot=<0.25 0.75>}
   slot {class=OverlaySlot, properties {horizontalAlignment=HAlign_Right, verticalAlignment=VAlign_Center, padding=<ltrb 0 0 90 0>}}
  }
  BackgroundBlur_1 {
   widgetClassPath=&BackgroundBlur
   parent=SizeBox_5
   widget=BackgroundBlur_1
   properties {blurStrength=16.4286, horizontalAlignment=HAlign_Right}
   slot {class=SizeBoxSlot, properties {horizontalAlignment=HAlign_Right}}
  }
  CommonBorder_0 {
   widgetClassPath=&CommonBorder
   parent=BackgroundBlur_1
   widget=CommonBorder_0
   properties {
    horizontalAlignment=HAlign_Right
    style=/FortUILibrary/Styles/Border/Border_Navy_VGrad.Border_Navy_VGrad_C
    verticalAlignment=VAlign_Center
    background {
     bIsDynamicallyLoaded=false
     drawAs=Image
     imageType=NoImage
     mirroring=NoMirror
     resourceName=None
     resourceObject=/FortUILibrary/Styles/Border/Materials/M_Border_Navy_VGrad
     tiling=NoTile
     imageSize=<32 32>
     margin=<ltrb 0 0 0 0>
     outlineSettings {
      bUseBrushTransparency=false
      roundingType=HalfHeightRadius
      width=0
      color {colorUseRule=UseColor_Specified, specifiedColor=<rgba 0 0 0 0>}
      cornerRadii=<0 0 0 0>
     }
     tintColor {colorUseRule=UseColor_Specified, specifiedColor=<rgba 1 1 1 1>}
     uVRegion {bIsValid=false, max=<0 0>, min=<0 0>}
    }
    brushColor=<rgba 1 1 1 0.5714>
    padding=<ltrb 0 0 0 0>
   }
   slot {class=BackgroundBlurSlot, properties {horizontalAlignment=HAlign_Right, padding=<ltrb 0 0 0 0>}}
  }
  GiftBoxImage {widgetClassPath=&CommonLazyImage, bInherited=true, widget=null}
 ]
}
