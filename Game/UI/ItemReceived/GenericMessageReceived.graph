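# Full-screen gift box message screen, based on FortMessageReceivedScreen, showing header, body, optional tournament subheader and Accept button.
# OnGiftBoxSet shows the Epic logo if flagged, resets the frontend camera, plays the pop-in animation and whoosh, then fills text.
# A shared SetText resolves header and body from override param, asset-string param, or definition default.
# Accept deactivates the widget; deactivation triggers gift navigation and broadcasts On Message Closed.
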WidgetBlueprint /Game/UI/ItemReceived/GenericMessageReceived
animations [Intro {
  binding_count=5
  display_label=""
  length_seconds=1.3001
  object=&self:Intro
  sequence {bindings [
    SubHeader {
     id=#1
     possessed_class=CommonTextBlock
     tracks [{
       type=MovieScene2DTransformTrack
       sections [{
         type=MovieScene2DTransformSection
         end_frame=16
         start_frame=0
         channels [
          Translation.X {type=MovieSceneScriptingFloatChannel}
          Translation.Y {type=MovieSceneScriptingFloatChannel}
          Angle {type=MovieSceneScriptingFloatChannel}
          Scale.X {type=MovieSceneScriptingFloatChannel, keys [[6.001,0],[11.001,1.2],[13.001,1]]}
          Scale.Y {type=MovieSceneScriptingFloatChannel, keys [[6.001,0],[11.001,1.2],[13.001,1]]}
          Shear.X {type=MovieSceneScriptingFloatChannel}
          Shear.Y {type=MovieSceneScriptingFloatChannel}
         ]
        }]
      }]
    }
    Btn_Accept {
     id=#2
     possessed_class=WBP_UIKit_Button_Loud_C
     tracks [
      {
       type=MovieSceneFloatTrack
       display_name=Render Opacity
       sections [{
         type=MovieSceneFloatSection
         end_frame=26
         start_frame=0
         channels [None {type=MovieSceneScriptingFloatChannel, keys [[0,0],[11.001,0],[26.001,1]]}]
        }]
      }
      {
       type=MovieScene2DTransformTrack
       sections [{
         type=MovieScene2DTransformSection
         end_frame=18
         start_frame=9
         channels [
          Translation.X {type=MovieSceneScriptingFloatChannel}
          Translation.Y {type=MovieSceneScriptingFloatChannel, thinned_from=4, keys [[9,50],[18.0007,0],[18.001,0]]}
          Angle {type=MovieSceneScriptingFloatChannel}
          Scale.X {type=MovieSceneScriptingFloatChannel}
          Scale.Y {type=MovieSceneScriptingFloatChannel}
          Shear.X {type=MovieSceneScriptingFloatChannel}
          Shear.Y {type=MovieSceneScriptingFloatChannel}
         ]
        }]
      }
     ]
    }
    Header {
     id=#3
     possessed_class=CommonTextBlock
     tracks [{
       type=MovieScene2DTransformTrack
       sections [{
         type=MovieScene2DTransformSection
         end_frame=16
         start_frame=0
         channels [
          Translation.X {type=MovieSceneScriptingFloatChannel}
          Translation.Y {type=MovieSceneScriptingFloatChannel}
          Angle {type=MovieSceneScriptingFloatChannel}
          Scale.X {type=MovieSceneScriptingFloatChannel, keys [[6.001,0],[11.001,1.2],[13.001,1]]}
          Scale.Y {type=MovieSceneScriptingFloatChannel, keys [[6.001,0],[11.001,1.2],[13.001,1]]}
          Shear.X {type=MovieSceneScriptingFloatChannel}
          Shear.Y {type=MovieSceneScriptingFloatChannel}
         ]
        }]
      }]
    }
    Body {
     id=#4
     possessed_class=CommonTextBlock
     tracks [{
       type=MovieScene2DTransformTrack
       sections [{
         type=MovieScene2DTransformSection
         end_frame=16
         start_frame=0
         channels [
          Translation.X {type=MovieSceneScriptingFloatChannel}
          Translation.Y {type=MovieSceneScriptingFloatChannel}
          Angle {type=MovieSceneScriptingFloatChannel}
          Scale.X {type=MovieSceneScriptingFloatChannel, keys [[6.001,0],[11.001,1.2],[13.001,1]]}
          Scale.Y {type=MovieSceneScriptingFloatChannel, keys [[6.001,0],[11.001,1.2],[13.001,1]]}
          Shear.X {type=MovieSceneScriptingFloatChannel}
          Shear.Y {type=MovieSceneScriptingFloatChannel}
         ]
        }]
      }]
    }
    SizeBox_4 {
     id=#5
     possessed_class=SizeBox
     tracks [{
       type=MovieScene2DTransformTrack
       sections [{
         type=MovieScene2DTransformSection
         end_frame=16
         start_frame=0
         channels [
          Translation.X {type=MovieSceneScriptingFloatChannel}
          Translation.Y {type=MovieSceneScriptingFloatChannel}
          Angle {type=MovieSceneScriptingFloatChannel}
          Scale.X {type=MovieSceneScriptingFloatChannel, keys [[6.001,0],[11.001,1.2],[13.001,1]]}
          Scale.Y {type=MovieSceneScriptingFloatChannel, keys [[6.001,0],[11.001,1.2],[13.001,1]]}
          Shear.X {type=MovieSceneScriptingFloatChannel}
          Shear.Y {type=MovieSceneScriptingFloatChannel}
         ]
        }]
      }]
    }
   ]}
 }]
graphs [
 EventGraph {nodes [
   n1 = Get ItemDisplayAsset!(self:Fort Gift Box Item Definition Object Reference=GetGiftBoxDefinition!(self:Fort Gift Box Item Object Reference=Get GiftBoxItem() -> (GiftBoxItem:Fort Gift Box Item Object Reference)) -> (ReturnValue:Fort Gift Box Item Definition Object Reference)) -> (ItemDisplayAsset:Soft Object Reference)
   n2 = Get CommonLazyImage_0() -> (CommonLazyImage_0:Common Lazy Image Object Reference)
   SetBrushFromLazyDisplayAsset!(execute, self:Common Lazy Image Object Reference=@n2, LazyObject:"Soft Object Reference (by ref)"=@n1, bMatchTextureSize:bool=false) -> (then)
   > n3 = SetVisibility(self=@n2, InVisibility=Select(Option 0=Collapsed, Option 1=SelfHitTestInvisible, Index=IsValidSoftObjectReference(SoftObjectReference=@n1)))
   SetFrontEndCamera!(
    execute=[@n3.then, @n8.then]
    self:Fort Front End Context Object Reference=Get FortFrontEndContext!() -> (ReturnValue:Fort Front End Context Object Reference) {is=title K2Node_GetBlueprintContextEx}
    NewState:EFrontEndCamera Enum=FrontendDefault
    TransitionParams:View Target Transition Params Structure=(BlendTime=0,BlendFunction=VTBlend_Cubic,BlendExp=2,bLockOutgoing=False)
   ) -> (then)
   > PlayAnimation(InAnimation=Get Intro())
   > SpawnSound2D!(
    execute
    Sound:Sound Base Object Reference=Get WhooshSound() -> (WhooshSound:Sound Base Object Reference)
    VolumeMultiplier:float=1
    PitchMultiplier:float=1
    StartTime:float=0
    ConcurrencySettings:Sound Concurrency Object Reference
    bPersistAcrossLevelTransition:bool=false
    bAutoDestroy:bool=true
   ) -> (then, ReturnValue:Audio Component Object Reference)
   > Set Subheader() {is=title K2Node_CallFunction}
   > n4 = Sequence()
   n5 = Get GiftBoxItem()
   > SetText!(
    execute
    self:Self Object Reference
    OverrideString:String=HeaderOverrideString
    AssetString:String=HeaderAssetString
    TextObject:Common Text Object Reference=Get Header() -> (Header:Common Text Object Reference)
    DefaultText:Text=Get DefaultHeaderText!(self:Fort Gift Box Item Definition Object Reference=GetGiftBoxDefinition(self=@n5)) -> (DefaultHeaderText:Text)
    bIsHeader:bool=true
   ) -> (then)
   n6 = Get GiftBoxItem()
   SetText(
    execute=@n4.then_1
    OverrideString=BodyOverrideString
    AssetString=BodyAssetString
    TextObject=Get Body() -> (Body:Common Text Object Reference)
    DefaultText=Get DefaultBodyText!(self:Fort Gift Box Item Definition Object Reference=GetGiftBoxDefinition(self=@n6)) -> (DefaultBodyText:Text)
    bIsHeader=false
   )
   "On Clicked (Btn_Accept)"!() -> (OutputDelegate, then, Button:Common Button Base Object Reference) {is=title K2Node_ComponentBoundEvent}
   > DeactivateWidget()
   Event On Deactivated()
   > NavigateForGift!(execute, Item:Fort Gift Box Item Object Reference=Get GiftBoxItem()) -> (then)
   > Call On Message Closed!(execute:Exec, self:Generic Message Received Object Reference) -> (then:Exec) {is=title K2Node_CallDelegate}
   Event OnGiftBoxSet() {is=title K2Node_Event}
   n7 = Get GiftBoxItem()
   > n8 = SetVisibility(
    self=Get CommonLazyImage_0()
    InVisibility=Select(
     Option 0=Collapsed
     Option 1=Visible
     Index=HasParam!(self:Fort Gift Box Item Object Reference=@n7, ParamName:String=EpicLogo) -> (ReturnValue:bool)
    )
   )
  ]}
 Set Subheader {nodes [
   Set Subheader() {is=title K2Node_FunctionEntry}
   > n1 = GetTournamentTitleFromCMS!(execute, self:Fort Message Received Screen Object Reference) -> (then, OutTitle:String, ReturnValue:bool)
   > n2 = Branch(Condition=@n1.ReturnValue)
   n3 = Get SubHeader() -> (SubHeader:Common Text Object Reference)
   > "SetText (Text)"(self=@n3, InText="To Text (String)"(InString=@n1))
   > SetVisibility(self=@n3, InVisibility=SelfHitTestInvisible)
   SetVisibility(execute=@n2.else, self=@n3, InVisibility=Collapsed)
  ]}
 SetText {nodes [
   n1 = SetText!() -> (then, OverrideString:String, AssetString:String, TextObject:Common Text Object Reference, DefaultText:Text, bIsHeader:bool=~) {is=title K2Node_FunctionEntry}
   n2 = Get GiftBoxItem()
   > n3 = Branch(Condition=HasParam(self=@n2, ParamName=@n1))
   > "SetText (Text)"(
    self=@n1.TextObject
    InText=FormatTextUsingParams!(
     self:Fort Gift Box Item Object Reference=@n2
     FormatString:"Text (by ref)"=Format Text!(Format="LOCTEXT(\"{overrideString}\")", overrideString:String=@n1) -> (Result)
    ) -> (ReturnValue:Text)
   )
   n4 = TryGetParamString!(self:Fort Gift Box Item Object Reference=@n2, ParamName:String=@n1.AssetString) -> (OutValue:String, ReturnValue:bool)
   n5 = Branch(execute=@n3.else, Condition=@n4.ReturnValue)
   > n6 = Branch(Condition=@n1.bIsHeader)
   > n7 = GetHeaderAssetString!(execute, self:Fort Message Received Screen Object Reference, Param:String=@n4) -> (then, ReturnValue:Text)
   > "SetText (Text)"(self=@n1.TextObject, InText=@n7)
   n8 = GetBodyAssetString!(execute=@n6.else, self:Fort Message Received Screen Object Reference, Param:String=@n4) -> (then, ReturnValue:Text)
   > "SetText (Text)"(self=@n1.TextObject, InText=@n8)
   "SetText (Text)"(execute=@n5.else, self=@n1.TextObject, InText=@n1.DefaultText)
  ]}
]
variables [WhooshSound {type=SoundBase, default=/Game/Sounds/Fort_UI_Sounds/FrontEnd/RewardsAndGifts/UI_FrontEnd_Gift_Generic_Message_Appear_Cue}]
widgets {
 info {inheritedWidgetCount=1, parentClass=/Script/FortniteUI.FortMessageReceivedScreen, rootWidgetClass=&Overlay, widgetCount=11}
 widgets [
  Overlay_2 {widgetClassPath=&Overlay, widget=Overlay_2}
  WBP_UIKit_Backplate {
   widgetClassPath=&WBP_UIKit_Backplate
   bIsVariable=true
   parent=Overlay_2
   widget=WBP_UIKit_Backplate
   properties {brightness=Dark}
   slot {class=OverlaySlot, properties {horizontalAlignment=HAlign_Fill, verticalAlignment=VAlign_Fill}}
  }
  MainSafeZone {
   widgetClassPath=&SafeZone
   parent=Overlay_2
   widget=MainSafeZone
   slot {class=OverlaySlot, properties {horizontalAlignment=HAlign_Fill, verticalAlignment=VAlign_Fill}}
  }
  MainContainer {
   widgetClassPath=&VerticalBox
   parent=MainSafeZone
   widget=MainContainer
   slot {class=SafeZoneSlot, properties {hAlign=HAlign_Center, vAlign=VAlign_Center}}
  }
  SizeBox_4 {
   widgetClassPath=&SizeBox
   parent=MainContainer
   widget=SizeBox_4
   properties {bOverride_HeightOverride=true, bOverride_WidthOverride=true, heightOverride=380, widthOverride=380}
   slot {class=VerticalBoxSlot, properties {horizontalAlignment=HAlign_Center, verticalAlignment=VAlign_Top, padding=<ltrb 0 0 0 8>}}
  }
  CommonLazyImage_0 {
   widgetClassPath=/Script/CommonUI.CommonLazyImage
   bIsVariable=true
   parent=SizeBox_4
   widget=CommonLazyImage_0
   properties {brush {
     bIsDynamicallyLoaded=false
     drawAs=Image
     imageType=NoImage
     mirroring=NoMirror
     resourceName=None
     resourceObject=/Game/UI/Login/Art/T_EpicLogo
     tiling=NoTile
     imageSize=<256 256>
     margin=<ltrb 0 0 0 0>
     outlineSettings {
      bUseBrushTransparency=false
      roundingType=HalfHeightRadius
      width=0
      color {colorUseRule=UseColor_Specified, specifiedColor=<rgba 0 0 0 0>}
      cornerRadii=<0 0 0 0>
     }
     tintColor {colorUseRule=UseColor_Specified, specifiedColor=<rgba 1 1 1 1>}
     uVRegion {bIsValid=false, max=<0 0>, min=<0 0>}
    }}
   slot {class=SizeBoxSlot}
  }
  Header {
   widgetClassPath=&CommonTextBlock
   bIsVariable=true
   parent=MainContainer
   widget=Header
   properties {
    justification=Center
    text="HEADER![ph]"
    wrapTextAt=850
    font {
     bForceMonospaced=false
     bMaterialIsStencil=false
     fontMaterial=null
     fontObject=&HeadingNow
     letterSpacing=-40
     monospacedWidth=1
     size=32.25
     skewAmount=0
     typefaceFontName=86 Bold Italic
     outlineSettings {
      bApplyOutlineToDropShadows=false
      bMiteredCorners=false
      bSeparateFillAlpha=false
      outlineMaterial=null
      outlineSize=0
      outlineColor=<rgba 0 0 0 0.07>
     }
    }
    shadowOffset=<0 0>
   }
   slot {class=VerticalBoxSlot, properties {horizontalAlignment=HAlign_Center, verticalAlignment=VAlign_Center}}
  }
  SubHeader {
   widgetClassPath=&CommonTextBlock
   bIsVariable=true
   parent=MainContainer
   widget=SubHeader
   properties {
    justification=Center
    text="SUBHEADER![ph]"
    wrapTextAt=850
    font {
     bForceMonospaced=false
     bMaterialIsStencil=false
     fontMaterial=null
     fontObject=&HeadingNow
     letterSpacing=0
     monospacedWidth=1
     size=20.25
     skewAmount=0
     typefaceFontName=75 Medium
     outlineSettings {
      bApplyOutlineToDropShadows=false
      bMiteredCorners=false
      bSeparateFillAlpha=false
      outlineMaterial=null
      outlineSize=0
      outlineColor=<rgba 0 0 0 0.07>
     }
    }
    shadowOffset=<0 0>
   }
   slot {class=VerticalBoxSlot, properties {horizontalAlignment=HAlign_Center, verticalAlignment=VAlign_Center, padding=<ltrb 0 8 0 0>}}
  }
  Body {
   widgetClassPath=&CommonTextBlock
   bIsVariable=true
   parent=MainContainer
   widget=Body
   properties {
    justification=Center
    text="battlepass body battlepass body battlepass body battlepass body battlepass body battlepass body battlepass body [ph]"
    wrapTextAt=850
    colorAndOpacity {colorUseRule=UseColor_Specified, specifiedColor=<rgba 0.5776 0.5776 0.5776 1>}
    font {
     bForceMonospaced=false
     bMaterialIsStencil=false
     fontMaterial=null
     fontObject=&HeadingNow
     letterSpacing=0
     monospacedWidth=1
     size=18
     skewAmount=0
     typefaceFontName=74 Regular
     outlineSettings {
      bApplyOutlineToDropShadows=false
      bMiteredCorners=false
      bSeparateFillAlpha=false
      outlineMaterial=null
      outlineSize=0
      outlineColor=<rgba 0 0 0 1>
     }
    }
    shadowOffset=<0 0>
   }
   slot {class=VerticalBoxSlot, properties {horizontalAlignment=HAlign_Center, verticalAlignment=VAlign_Center, padding=<ltrb 0 8 0 0>}}
  }
  Btn_Accept {
   widgetClassPath=&WBP_UIKit_Button_Regular
   bIsVariable=true
   parent=MainContainer
   widget=Btn_Accept
   properties {text=Accept, triggeringInputAction {dataTable=&InputActionDataTable, rowName=GenericForward}}
   slot {class=VerticalBoxSlot, properties {horizontalAlignment=HAlign_Center, verticalAlignment=VAlign_Bottom, padding=<ltrb 0 48 0 0>}}
  }
  GiftBoxImage {widgetClassPath=&CommonLazyImage, bInherited=true, widget=null}
 ]
}
