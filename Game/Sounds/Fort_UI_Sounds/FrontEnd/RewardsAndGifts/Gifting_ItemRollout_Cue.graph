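# Sound cue for the front-end gifting item rollout, played as a UI sound in the Fort_UI sound class.
# A single wave player feeds the Gift_Receive_01 wave straight to the output with boosted volume.
# A concurrency override caps simultaneous instances at a small count and stops the oldest when exceeded.
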SoundCue /Game/Sounds/Fort_UI_Sounds/FrontEnd/RewardsAndGifts/Gifting_ItemRollout_Cue
root=WavePlayer_1
nodes [{id=WavePlayer_1, properties {soundWaveAssetPtr=/Game/Sounds/Fort_UI_Sounds/FrontEnd/Gift_Receive_01}}]
properties {
 bOverrideConcurrency=true
 maxDistance=2097152
 soundClassObject=&Fort_UI
 volumeMultiplier=1.5
 concurrencyOverrides {maxCount=2, resolutionRule=StopOldest}
}
