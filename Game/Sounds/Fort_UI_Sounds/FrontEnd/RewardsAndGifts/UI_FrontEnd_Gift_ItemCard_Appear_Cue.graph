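# UI sound cue for a gift item card appearing in the front end, routed to the Fort_UI sound class.
# A root Mixer layers four wave players at equal volume: a battle pass unbox sound, a gifting item-shown sound and the same medium modal whoosh twice.
# Concurrency override caps it at two instances, stopping the oldest, in place of the assigned front-end UI concurrency set.
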SoundCue /Game/Sounds/Fort_UI_Sounds/FrontEnd/RewardsAndGifts/UI_FrontEnd_Gift_ItemCard_Appear_Cue
root=Mixer_0
nodes [
 {id=Mixer_0, children [WavePlayer_2,WavePlayer_0,WavePlayer_4,WavePlayer_3], properties {inputVolume [1,1,1,1]}}
 {id=WavePlayer_0, properties {soundWaveAssetPtr=/Game/Sounds/Fort_UI_Sounds/FrontEnd/fort_battlepass_unboxtiers}}
 {id=WavePlayer_2, properties {soundWaveAssetPtr=/Game/Sounds/Fort_UI_Sounds/FrontEnd/BattlePass/BP_ToolKit/BP_Modal_Whoosh_Med_03}}
 {id=WavePlayer_3, properties {soundWaveAssetPtr=/Game/Sounds/Fort_UI_Sounds/FrontEnd/New_FrontEnd/gifting_itemshown_01}}
 {id=WavePlayer_4, properties {soundWaveAssetPtr=&BP_Modal_Whoosh_Med_03}}
]
properties {
 bOverrideConcurrency=true
 maxDistance=2097152
 soundClassObject=&Fort_UI
 volumeMultiplier=1
 concurrencyOverrides {maxCount=2, resolutionRule=StopOldest}
 concurrencySet [/Game/Sounds/Concurrency_Rules/FrontEnd_UI_Menu_Max_03_Stop_Oldest]
}
