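# Button widget that shows one received item as a large card on an item-received screen.
# On construct it feeds the loot info's item instance and quantity into a multi-size item card.
# PlayShow runs a scale-pop and wobble animation, playing the appear sound only for indices below four.
# Hover, focus and OnCentered broadcast the loot info through OnHoveredEvent, and received focus is forwarded to the inner card.
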WidgetBlueprint /Game/UI/ItemReceived/ItemCardWidget
animations [Show {
  binding_count=1
  display_label=""
  length_seconds=0.4
  object=&self:Show
  sequence {
   bindings [ItemCardWidget_C_49 {
     id=#1
     possessed_class=ItemCardWidget_C
     tracks [{
       type=MovieScene2DTransformTrack
       sections [{
         type=MovieScene2DTransformSection
         end_frame=400
         start_frame=0
         channels [
          Translation.X {type=MovieSceneScriptingFloatChannel}
          Translation.Y {type=MovieSceneScriptingFloatChannel}
          Angle {type=MovieSceneScriptingFloatChannel, keys [[72.7167,0],[181.8167,6],[327.2667,0]]}
          Scale.X {type=MovieSceneScriptingFloatChannel, keys [[0,0],[109.0833,1.5],[300,0.93],[400,1]]}
          Scale.Y {type=MovieSceneScriptingFloatChannel, keys [[0,0],[109.0833,1.5],[300,0.93],[400,1]]}
          Shear.X {type=MovieSceneScriptingFloatChannel}
          Shear.Y {type=MovieSceneScriptingFloatChannel}
         ]
        }]
      }]
    }]
   display_rate {numerator=1000}
  }
 }]
graphs [
 EventGraph {nodes [
   n1 = OnCentered() {is=title K2Node_CustomEvent}
   Call On Hovered Event!(
    execute:Exec=[@n1.then, @n4.then]
    self:Item Card Widget Object Reference
    NewParam:Fort Received Item Loot Info Structure=Get ItemReceivedInfo_BP() -> (ItemReceivedInfo_BP:Fort Received Item Loot Info Structure)
   ) -> (then:Exec) {is=title K2Node_CallDelegate}
   n2 = PlayShow!() -> (OutputDelegate, then, Idx:int) {is=title K2Node_CustomEvent}
   > PlayAnimation(InAnimation=Get Show() -> (Show:Widget Animation Object Reference))
   > Branch(Condition="integer < integer"!(A:int=@n2.Idx, B:int=4) -> (ReturnValue:bool) {is=title K2Node_CallFunction})
   > PlaySound2D(Sound=Get SoundOnItemCardShow() -> (SoundOnItemCardShow:Sound Base Object Reference))
   Event Construct()
   n3 = Break Fort Received Item Loot Info!(FortReceivedItemLootInfo:"Fort Received Item Loot Info Structure (by ref)"=Get ItemReceivedInfo_BP()) -> (
    HeaderInformation:Fort Item Header Information Structure
    GeneratedItemInstance:Fort Item Object Reference
    ItemDef:Fort Item Definition Object Reference
    ItemVM:Fort Item VM Object Reference
    TemplateId:String
    Quantity:int
   ) {is=title K2Node_BreakStruct}
   > SetItemToRepresent!(
    execute
    self:Fort Item Widget NUI Object Reference=Get FortMultiSizeItemCard_0() -> (FortMultiSizeItemCard_0:Fort Multi Size Item Card Object Reference)
    ItemToRepresent:Fort Item Object Reference=@n3.GeneratedItemInstance
   ) -> (then)
   > SetQuantityOverride!(execute, self:Fort Multi Size Item Card Object Reference=Get FortMultiSizeItemCard_0(), QuantityOverride:int=@n3.Quantity) -> (then)
   n4 = Event On Hovered()
   n5 = GetDisplayName!(self:Item Definition Base Object Reference=Break Fort Received Item Loot Info(FortReceivedItemLootInfo=Get ItemReceivedInfo_BP()).ItemDef) -> (ReturnValue:Text)
   PrintString!(execute, InString=Append(A="To String (Text)"(InText=@n5), B=" Lost"), bPrintToScreen, bPrintToLog, TextColor, Duration, Key) -> (then)
   Event PreConstruct()
   Event Tick()
   Event OnFocusLost!() -> (OutputDelegate, then, InFocusEvent:Focus Event Structure) {is=title K2Node_Event}
  ]}
 OnFocusReceived {nodes [
   n1 = Break Fort Received Item Loot Info(FortReceivedItemLootInfo=Get ItemReceivedInfo_BP())
   n2 = PrintString(InString="To String (Text)"(InText=GetDisplayName(self=@n1.ItemDef)))
   Call On Hovered Event(execute=[@n2.then, @n3.then])
   > Return Node!(
    execute
    ReturnValue:Event Reply Structure=SetUserFocus!(Reply:"Event Reply Structure (by ref)"=Handled(), FocusWidget:Widget Object Reference=Get FortMultiSizeItemCard_0(), bInAllUsers:bool=false) -> (ReturnValue:Event Reply Structure)
   )
   n3 = OnFocusReceived!() -> (then, MyGeometry:Geometry Structure, InFocusEvent:Focus Event Structure) {is=title K2Node_FunctionEntry}
  ]}
 OnHoveredEvent {nodes [OnHoveredEvent!() -> (then, NewParam:Fort Received Item Loot Info Structure) {is=title K2Node_FunctionEntry}]}
]
variables [
 ItemReceivedInfo_BP {
  type=FortReceivedItemLootInfo
  default {generatedItemInstance=null, itemDef=null, itemVM=null, quantity=0, templateId="", headerInformation {}}
 }
 OnHoveredEvent {}
 SoundOnItemCardShow {type=SoundBase, default=/Game/Sounds/Fort_UI_Sounds/FrontEnd/RewardsAndGifts/UI_FrontEnd_Gift_ItemCard_Appear_Cue}
]
widgets {
 info {inheritedWidgetCount=1, parentClass=&CommonButtonLegacy, rootWidgetClass=&VerticalBox, widgetCount=3}
 widgets [
  VerticalBox_2 {widgetClassPath=&VerticalBox, widget=VerticalBox_2}
  FortMultiSizeItemCard_0 {
   widgetClassPath=/Script/FortniteUI.FortMultiSizeItemCard
   bIsVariable=true
   parent=VerticalBox_2
   widget=FortMultiSizeItemCard_0
   properties {bPItemCardSize=XL, designTimePreviewItemType=/BRCosmetics/Athena/Items/Weapons/WID_Harvest_Pickaxe_Shark_Athena}
   slot {class=VerticalBoxSlot, properties {horizontalAlignment=HAlign_Center, verticalAlignment=VAlign_Center}}
  }
  InputActionWidget {widgetClassPath=&CommonActionWidget, bInherited=true, widget=null}
 ]
}
