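# Gift box "items received" screen that reveals each granted item as a card in a horizontal scroll box, previewing it in the front-end Cosmetics camera.
# Setting the gift box loads a header sub-widget from its definition, then items show click-by-click (UseClickRollout) or via a timed rollout.
# Selection updates the vault view, info panel and unowned-base-item warning for cosmetics and style tokens.
# Close plays a Collected animation, then pops the screen and broadcasts gifting completion.
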WidgetBlueprint /Game/UI/ItemReceived/ItemReceived_Base
animations [
 Intro {
  binding_count=2
  display_label=""
  length_seconds=0.8
  object=&self:Intro
  sequence {bindings [
    ItemReceived_Base_C_26 {
     id=#1
     possessed_class=ItemReceived_Base_C
     tracks [
      {type=MovieSceneFloatTrack, display_name=Render Opacity, sections []}
      {
       type=MovieScene2DTransformTrack
       sections [{
         type=MovieScene2DTransformSection
         end_frame=0
         start_frame=0
         channels [
          Translation.X {type=MovieSceneScriptingFloatChannel, const [0,1]}
          Translation.Y {type=MovieSceneScriptingFloatChannel, const [0,1]}
          Angle {type=MovieSceneScriptingFloatChannel, const [0,1]}
          Scale.X {type=MovieSceneScriptingFloatChannel, const [1,1]}
          Scale.Y {type=MovieSceneScriptingFloatChannel, const [1,1]}
          Shear.X {type=MovieSceneScriptingFloatChannel, const [0,1]}
          Shear.Y {type=MovieSceneScriptingFloatChannel, const [0,1]}
         ]
        }]
      }
     ]
    }
    LeftContainer {
     id=#2
     possessed_class=VerticalBox
     tracks [{
       type=MovieScene2DTransformTrack
       sections [{
         type=MovieScene2DTransformSection
         end_frame=16
         start_frame=-1
         channels [
          Translation.X {type=MovieSceneScriptingFloatChannel}
          Translation.Y {type=MovieSceneScriptingFloatChannel}
          Angle {type=MovieSceneScriptingFloatChannel}
          Scale.X {type=MovieSceneScriptingFloatChannel, keys [[0,0],[6,1.2],[8,1]]}
          Scale.Y {type=MovieSceneScriptingFloatChannel, keys [[0,0],[6,1.2],[8,1]]}
          Shear.X {type=MovieSceneScriptingFloatChannel}
          Shear.Y {type=MovieSceneScriptingFloatChannel}
         ]
        }]
      }]
    }
   ]}
 }
 Collected {
  binding_count=3
  display_label=""
  length_seconds=0.5
  object=&self:Collected
  sequence {bindings [
    TextBlock_0 {
     id=#3
     possessed_class=TextBlock
     tracks [
      {
       type=MovieSceneByteTrack
       display_name=Visibility
       sections [{type=MovieSceneByteSection, end_frame=0, start_frame=0, channels [None {type=MovieSceneScriptingByteChannel, const [0,1]}]}]
      }
      {
       type=MovieScene2DTransformTrack
       sections [{
         type=MovieScene2DTransformSection
         end_frame=10
         start_frame=0
         channels [
          Translation.X {type=MovieSceneScriptingFloatChannel, const [0,1]}
          Translation.Y {type=MovieSceneScriptingFloatChannel, const [0,1]}
          Angle {type=MovieSceneScriptingFloatChannel, const [0,1]}
          Scale.X {type=MovieSceneScriptingFloatChannel, keys [[5.0003,0],[8.0003,1.2],[10.0003,1]]}
          Scale.Y {type=MovieSceneScriptingFloatChannel, keys [[5.0003,0],[8.0003,1.1],[10.0003,1]]}
          Shear.X {type=MovieSceneScriptingFloatChannel, const [0,1]}
          Shear.Y {type=MovieSceneScriptingFloatChannel, const [0,1]}
         ]
        }]
      }
     ]
    }
    Text_ItemCount {
     id=#4
     possessed_class=CommonTextBlock
     tracks [{
       type=MovieSceneFloatTrack
       display_name=Render Opacity
       sections [{type=MovieSceneFloatSection, end_frame=5, start_frame=0, channels [None {type=MovieSceneScriptingFloatChannel, keys [[0,1],[5,0]]}]}]
      }]
    }
    HorizontalBoxActions {
     id=#5
     possessed_class=HorizontalBox
     tracks [
      {
       type=MovieScene2DTransformTrack
       sections [{
         type=MovieScene2DTransformSection
         end_frame=5
         start_frame=0
         channels [
          Translation.X {type=MovieSceneScriptingFloatChannel, const [0,1]}
          Translation.Y {type=MovieSceneScriptingFloatChannel, const [0,1]}
          Angle {type=MovieSceneScriptingFloatChannel, const [0,1]}
          Scale.X {type=MovieSceneScriptingFloatChannel, keys [[0,1],[2,1.1],[5,0]]}
          Scale.Y {type=MovieSceneScriptingFloatChannel, keys [[0,1],[2,1.2],[5,0]]}
          Shear.X {type=MovieSceneScriptingFloatChannel, const [0,1]}
          Shear.Y {type=MovieSceneScriptingFloatChannel, const [0,1]}
         ]
        }]
      }
      {
       type=MovieSceneByteTrack
       display_name=Visibility
       sections [{type=MovieSceneByteSection, end_frame=5, start_frame=0, channels [None {type=MovieSceneScriptingByteChannel, keys [[0,4],[5,2]]}]}]
      }
      {
       type=MovieSceneFloatTrack
       display_name=Render Opacity
       sections [{type=MovieSceneFloatSection, end_frame=7, start_frame=5, channels [None {type=MovieSceneScriptingFloatChannel, keys [[5,1],[7,0]]}]}]
      }
     ]
    }
   ]}
 }
 Fade {
  binding_count=1
  display_label=""
  length_seconds=5
  object=&self:Fade
  sequence {bindings [MainSafeZone {id=#6, possessed_class=SafeZone, tracks []}]}
 }
]
graphs [
 AddItemCard {nodes [
   n1 = AddItemCard!() -> (then, NewParam:Fort Received Item Loot Info Structure) {is=title K2Node_FunctionEntry}
   > n2 = Create Widget!(
    execute:Exec
    Class:User Widget Class Reference=&/ItemCardWidget.ItemCardWidget_C
    OwningPlayer:Player Controller Object Reference
    ItemReceivedInfo_BP:Fort Received Item Loot Info Structure=@n1
    Style:Common Button Style Class Reference=/FortUILibrary/Foundation/Buttons/ButtonStyles/ButtonStyle-Solid-SquareM.ButtonStyle-Solid-SquareM_C
    bApplyAlphaOnDisable:bool=False
    bLocked:bool=False
    bSelectable:bool=False
    bShouldSelectUponReceivingFocus:bool=False
    bInteractableWhenSelected:bool=False
    bToggleable:bool=False
    bTriggerClickedAfterSelection:bool=False
    bDisplayInputActionWhenNotInteractable:bool=True
    bHideInputActionWithKeyboard:bool=False
    bShouldUseFallbackDefaultInputAction:bool=True
    bRequiresHold:bool=False
    HoldData:Common UIHold Data Class Reference
    ClickMethod:EButtonClickMethod Enum=DownAndUp
    InputPriority:int=0
    TriggeringInputAction:Data Table Row Handle Structure
   ) -> (then:Exec, ReturnValue:Item Card Widget Object Reference)
   > Bind Event to On Hovered Event(self:Item Card Widget Object Reference=@n2, Delegate=Create Event()) {is=title K2Node_AddDelegate}
   > n3 = AddChild(self=Get GiftScrollBox() -> (GiftScrollBox:Scroll Box Object Reference), Content=@n2)
   > n4 = Cast To ScrollBoxSlot!(execute, Object=@n3) -> (then, CastFailed, AsScroll Box Slot:Scroll Box Slot Object Reference) {is=title K2Node_DynamicCast}
   > SetPadding!(execute, self:Scroll Box Slot Object Reference=@n4, InPadding:Margin Structure=Make Margin(Left=8, Top=0, Right=8, Bottom=0)) -> (then)
   n5 = Get GiftScrollBox()
   > SetAlignmentOnSlot!(
    execute
    self:Fort Item Received Widget Base Object Reference
    Widget:Scroll Box Object Reference=Get GiftScrollBox()
    Index:int=int - int(A=GetChildrenCount!(self:Panel Widget Object Reference=@n5) -> (ReturnValue:int=0), B=1) {is=title K2Node_CallFunction}
   ) -> (then)
   > SetCardSize!(
    execute
    self:Fort Multi Size Item Card Object Reference=Get FortMultiSizeItemCard_0!(self:Item Card Widget Object Reference=@n2) -> (FortMultiSizeItemCard_0:Fort Multi Size Item Card Object Reference)
    CardSize:EFortItemCardSize Enum=L
   ) -> (then)
   > Return Node!(execute, ReturnValue:Item Card Widget Object Reference=@n2)
  ]}
 EventGraph {nodes [
   n1 = Get ItemDefs() -> (ItemDefs:Array of Fort Received Item Loot Info Structures)
   n2 = "integer >= integer"!(
    A:int=Get CurrentIndex() -> (CurrentIndex:int=0)
    B:int=int - int(A=Length(TargetArray:Array of Fort Received Item Loot Info Structures=@n1), B=1)
   ) -> (ReturnValue:bool) {is=title K2Node_CallFunction}
   n3 = SetVisibility(
    self=Get CloseButton() -> (CloseButton:Icon Text Button Object Reference)
    InVisibility=Select(Option 0=Collapsed, Option 1=SelfHitTestInvisible, Index=@n2)
   )
   SetVisibility(
    execute=[@n3.then, @n17.then, @n45.then]
    self=Get Text_ItemCount() -> (Text_ItemCount:Common Text Object Reference)
    InVisibility=Select(Option 0=SelfHitTestInvisible, Option 1=Hidden, Index=@n2)
   )
   > SetVisibility(
    self=Get ShowNextButton() -> (ShowNextButton:Icon Text Button Object Reference)
    InVisibility=Select(Option 0=SelfHitTestInvisible, Option 1=Collapsed, Index=@n2)
   )
   > n4 = AddItemCard!(
    execute
    self
    NewParam:Fort Received Item Loot Info Structure="Get (a ref)"(Array:Array of Fort Received Item Loot Info Structures=Get ItemDefs(), Dimension 1=Get CurrentIndex() -> (CurrentIndex=0)) -> (Output:"Fort Received Item Loot Info Structure (by ref)")
   ) -> (then, ReturnValue:Item Card Widget Object Reference) {is=title K2Node_CallFunction}
   > PlayShow!(execute, self:Item Card Widget Object Reference=@n4, Idx:int=0) -> (then)
   n5 = Get ItemDefs()
   n6 = Get CurrentIndex() -> (CurrentIndex=0)
   > "SetText (Text)"(
    self=Get Text_ItemCount()
    InText=Format Text!(
     Format="LOCTEXT(\"Item {a} of {b}\")"
     a:int=int + int(A=@n6, B=1) -> (ReturnValue=0) {is=title K2Node_CommutativeAssociativeBinaryOperator}
     b:int=Length(TargetArray=@n5)
    ) -> (Result)
   )
   n7 = Get GiftScrollBox()
   > SelectItem!(execute, self, Index:int=int - int(A=GetChildrenCount(self=@n7) -> (ReturnValue=0), B=1), CenterWidget:bool=true) -> (then)
   n8 = "Get (a ref)"(Array=Get ItemDefs(), Dimension 1=Get CurrentIndex() -> (CurrentIndex=0))
   > UpdateFromItemReceivedInfo!(execute, self, Loot:Fort Received Item Loot Info Structure=@n8, select:bool=true) -> (then)
   > PopulateUsingItem!(
    execute
    self:Fort Item Info Widget Object Reference=Get ItemInfoWidget() -> (ItemInfoWidget:Item Info Widget Object Reference)
    NewItem:Fort Item Object Reference=Break Fort Received Item Loot Info!(FortReceivedItemLootInfo:"Fort Received Item Loot Info Structure (by ref)"=@n8) -> (
     HeaderInformation:Fort Item Header Information Structure
     GeneratedItemInstance:Fort Item Object Reference
     ItemDef:Fort Item Definition Object Reference
     ItemVM:Fort Item VM Object Reference
     TemplateId:String
     Quantity:int
    ) {is=title K2Node_BreakStruct}.GeneratedItemInstance
   ) -> (then)
   > Increment Int(Value=Get CurrentIndex() -> (CurrentIndex=0))
   n9 = Set Timer by Event(Delegate=Create Event(), Time=Get RolloutTime() -> (RolloutTime:double), bLooping=false)
   n10 = Cast To ItemCardWidget!(
    execute=[@n9.then, @n50.then]
    Object=GetChildAt!(self:Panel Widget Object Reference=Get GiftScrollBox(), Index:int=Get AnimationIndex() -> (AnimationIndex:int=0)) -> (ReturnValue:Widget Object Reference)
   ) -> (then, CastFailed, AsItem Card Widget:Item Card Widget Object Reference) {is=title K2Node_DynamicCast}
   > PlayShow(self=@n10, Idx=Get AnimationIndex() -> (AnimationIndex=0))
   > n11 = SetVisibility(self=@n10, InVisibility=SelfHitTestInvisible)
   Set AnimationIndexCached(execute=[@n11.then, @n10.CastFailed], AnimationIndexCached:int=Get AnimationIndex() -> (AnimationIndex=0)) -> (Output_Get:int=0)
   > n12 = Increment Int(Value=Get AnimationIndex() -> (AnimationIndex=0))
   > Set AnimationIndex(AnimationIndex:int=@n12) -> (Output_Get:int=0)
   n13 = "Equal (Integer)"(A=Get AnimationIndexCached() -> (AnimationIndexCached:int=0), B=0) {is=title K2Node_CallFunction}
   > Branch(Condition=@n13)
   > SelectItem(Index=0, CenterWidget=@n13)
   n14 = AnimateRollout()
   PlayAnimation(execute=[@n28.then, @n14.then], InAnimation=Get Intro())
   > SetAlignmentOnSlots!(execute, self:Fort Item Received Widget Base Object Reference=Self-Reference(), Widget:Scroll Box Object Reference=Get GiftScrollBox()) -> (then)
   > PlaySound2D(Sound=Get SoundOnItemReceived() -> (SoundOnItemReceived:Sound Base Object Reference))
   "On Button Clicked (EquipButton)"!() -> (OutputDelegate, then, Button:Common Button Legacy Object Reference) {is=title K2Node_ComponentBoundEvent}
   n15 = Get FortInventoryContext!() -> (ReturnValue:Fort Inventory Context Object Reference) {is=title K2Node_GetBlueprintContextEx}
   > n16 = Cast To FortAccountItem!(
    execute
    Object="Get (a copy)"(Array:Array of Fort Item Object References=GetItemInstancesByDefinition!(self:Fort Inventory Context Object Reference=@n15, ItemDefinition:Item Definition Base Object Reference) -> (Items:Array of Fort Item Object References)) -> (Output:Fort Item Object Reference)
   ) -> (then, CastFailed, AsFort Account Item:Fort Account Item Object Reference) {is=title K2Node_DynamicCast}
   > SetItemFavoriteStatus!(
    execute
    self:Fort Mcp Context Object Reference=Get FortMcpContext!() -> (ReturnValue:Fort Mcp Context Object Reference) {is=title K2Node_GetBlueprintContextEx}
    Item:Fort Account Item Object Reference=@n16
    bFavorite:bool=NOT Boolean(A=IsFavorite!(self:Fort Account Item Object Reference=@n16) -> (ReturnValue:bool))
    ShouldDeferMarking:bool=false
   ) -> (then)
   n17 = "On Button Clicked (ShowNextButton)"!() -> (OutputDelegate, then, Button:Common Button Legacy Object Reference) {is=title K2Node_ComponentBoundEvent}
   "On Button Clicked (CloseButton)"!() -> (OutputDelegate, then, Button:Common Button Legacy Object Reference) {is=title K2Node_ComponentBoundEvent}
   > Do Once()
   > ClosingGiftbox(self:Fort Item Received Widget Base Object Reference)
   > PlayAnimation(InAnimation=Get Collected() -> (Collected:Widget Animation Object Reference))
   > Delay(Duration=1.7)
   > HandleBack!(execute, self) -> (then, Passthrough:bool)
   n18 = Rollout Items Received() {is=title K2Node_CustomEvent}
   n19 = TryLoadHeaderSubWidgetClass!(
    execute=[@n59.then, @n18.then]
    self:Fort Gift Box Item Definition Object Reference=GetGiftBoxDefinition!(self:Fort Gift Box Item Object Reference=Get GiftBoxItem() -> (GiftBoxItem:Fort Gift Box Item Object Reference)) -> (ReturnValue:Fort Gift Box Item Definition Object Reference)
   ) -> (then, OutWidgetClass:User Widget Class Reference, ReturnValue:bool)
   > n20 = Create Widget!(execute, Class=@n19, OwningPlayer) -> (then, ReturnValue:User Widget Object Reference)
   > n21 = Cast To ItemReceivedHeaderSubWidgetBase!(execute, Object=@n20) -> (then, CastFailed, AsItem Received Header Sub Widget Base:Item Received Header Sub Widget Base Object Reference) {is=title K2Node_DynamicCast}
   > Set bSelfGift!(execute, bSelfGift:bool=Get FromSelf() -> (FromSelf:bool), self:Item Received Header Sub Widget Base Object Reference=@n21) -> (then, Output_Get:bool)
   > Set ItemDefs!(
    execute
    ItemDefs:Array of Fort Received Item Loot Info Structures=Get ItemDefs()
    self:Item Received Header Sub Widget Base Object Reference=@n21
   ) -> (then, Output_Get:Array of Fort Received Item Loot Info Structures)
   > Set GiftBoxItem_BP!(
    execute
    GiftBoxItem_BP:Fort Gift Box Item Object Reference=Get GiftBoxItem()
    self:Item Received Header Sub Widget Base Object Reference=@n21
   ) -> (then, Output_Get:Fort Gift Box Item Object Reference)
   > Set FromUserName_BP!(execute, FromUserName_BP:String=Get FromUsername() -> (FromUsername:String), self:Item Received Header Sub Widget Base Object Reference=@n21) -> (then, Output_Get:String)
   > InitFromGiftBoxItem_BP(self:Item Received Header Base Object Reference=@n21)
   > AddChild(self=Get HeaderContainer() -> (HeaderContainer:Horizontal Box Object Reference), Content=@n21)
   > n22 = Cast To ItemReceivedMissionCompleteHeaderLarge!(execute, Object=@n21) -> (then, CastFailed, AsItem Received Mission Complete Header Large:Item Received Mission Complete Header Large Object Reference) {is=title K2Node_DynamicCast}
   n23 = Get Image_DarkBackground() -> (Image_DarkBackground:Image Object Reference)
   > SetVisibility(self=@n23, InVisibility=HitTestInvisible)
   n24 = SlotAsHorizontalBoxSlot!(Widget:Widget Object Reference=Get SizeBox_Contents() -> (SizeBox_Contents:Size Box Object Reference)) -> (ReturnValue:Horizontal Box Slot Object Reference)
   > n25 = SetPadding(self:Horizontal Box Slot Object Reference=@n24, InPadding=Make Margin(Left=0, Top=0, Right=0, Bottom=0))
   UpdateRarityFlare!(execute=[@n30.then, @n25.then], self, NewParam:bool=false, Who:int=0) -> (then)
   > UpdateFromItemReceivedInfo(Loot="Get (a ref)"(Array=Get ItemDefs(), Dimension 1=0), select=false)
   > n26 = Branch(Condition=Get UseClickRollout() -> (UseClickRollout:bool))
   > ShowNextItem()
   n27 = For Each Loop(Exec=@n26.else, Array:Array of Fort Received Item Loot Info Structures=Get ItemDefs()) -> (Array Element:Fort Received Item Loot Info Structure)
   n28 = Set Timer by Event(execute=@n27.Completed, Delegate=Create Event(), Time=0.2, bLooping=false)
   n29 = AddItemCard(execute=@n27.LoopBody, NewParam=@n27)
   > SetVisibility(self=@n29, InVisibility=Hidden)
   SetVisibility(execute=@n22.CastFailed, self=@n23, InVisibility=Hidden)
   > n30 = SetPadding(self=@n24, InPadding=Make Margin(Left=50, Top=90, Right=0, Bottom=0))
   n31 = SelectItem!() -> (OutputDelegate, then, Index:int, CenterWidget:bool=~) {is=title K2Node_CustomEvent}
   n32 = Get GiftScrollBox()
   n33 = GetChildAt(self=@n32, Index=@n31.Index)
   > Is Valid(InputObject=@n33)
   > ScrollWidgetIntoView!(
    execute
    self:Scroll Box Object Reference=@n32
    WidgetToFind:Widget Object Reference=@n33
    AnimateScroll:bool=true
    ScrollDestination:EDescendantScrollDestination Enum=IntoView
    Padding:float=0
   ) -> (then)
   > n34 = Branch(Condition=@n31.CenterWidget)
   > CenterWidget!(execute, ToCenter:Widget Object Reference=@n33) -> (then)
   > n35 = Cast To ItemCardWidget(Object=@n33)
   > OnCentered(self:Item Card Widget Object Reference=@n35)
   n36 = "Get (a ref)"(Array=Get ItemDefs(), Dimension 1=@n31.Index)
   > n37 = UpdateFromItemReceivedInfo(Loot=@n36, select=true)
   PopulateUsingItem(
    execute=[@n37.then, @n34.else]
    self=Get ItemInfoWidget()
    NewItem=Break Fort Received Item Loot Info(FortReceivedItemLootInfo=@n36).GeneratedItemInstance
   )
   n38 = UpdateRarityFlare!() -> (OutputDelegate, then, NewParam:bool=~, Who:int) {is=title K2Node_CustomEvent}
   n39 = "integer >= integer"(A=@n38.Who, B=1)
   n40 = "Equal (Integer)"(A=Length(TargetArray=Get ItemDefs()), B=1)
   > SetVisibility(
    self=Get ImageRarityFlare() -> (ImageRarityFlare:Image Object Reference)
    InVisibility=Select(Option 0=Collapsed, Option 1=HitTestInvisible, Index=OR Boolean(A=@n39, B=@n40))
   )
   > Branch(Condition=@n40)
   > Sequence()
   n41 = "Get (a copy)"(Array:Array of Fort Received Item Loot Info Structures=Get ItemDefs()) -> (Output:Fort Received Item Loot Info Structure)
   > n42 = GetRarityData!(
    execute
    Rarity:EFortRarity Enum=GetRarity!(self:Fort Item Object Reference=Break Fort Received Item Loot Info(FortReceivedItemLootInfo=@n41).GeneratedItemInstance) -> (ReturnValue:EFortRarity Enum=Common)
   ) -> (then, ReturnValue:Fort Rarity Item Data Structure)
   > n43 = GetDynamicMaterial(self=Get ImageRarityFlare())
   n44 = Break Fort Rarity Item Data!(FortRarityItemData:"Fort Rarity Item Data Structure (by ref)"=@n42) -> (
    Name:Text
    Color1:Linear Color Structure
    Color2:Linear Color Structure
    Color3:Linear Color Structure
    Color4:Linear Color Structure
    Color5:Linear Color Structure
    Radius:float=~
    Falloff:float=~
    Brightness:float=~
    Roughness:float=~
    Glow:float=~
   ) {is=title K2Node_BreakStruct}
   > SetVectorParameterValue(self=@n43, ParameterName=Rarity Color 1, Value=@n44.Color2)
   > SetVectorParameterValue(self=@n43, ParameterName=Rarity Color 2, Value=@n44.Color3)
   n45 = ShowNextItem() {is=title K2Node_CustomEvent}
   AnimateRollout() {is=title K2Node_CustomEvent}
   > n46 = Branch(Condition="integer > integer"(A=Get AnimationIndex() -> (AnimationIndex=0), B=1) {is=title K2Node_CallFunction})
   > n47 = SetScrollBarVisibility!(execute, self:Scroll Box Object Reference=Get GiftScrollBox(), NewScrollBarVisibility:ESlateVisibility Enum=Visible) -> (then)
   n48 = Get AnimationIndex() -> (AnimationIndex=0)
   n49 = Branch(
    execute=[@n47.then, @n46.else]
    Condition=Is Valid Index!(TargetArray:Array of Fort Received Item Loot Info Structures=Get ItemDefs(), IndexToTest:int=@n48) -> (ReturnValue:bool)
   )
   > n50 = SelectItem(Index=@n48, CenterWidget=true) {is=title K2Node_CallFunction}
   n51 = Get GiftScrollBox()
   SetScrollBarVisibility(execute=[@n60.then, @n49.else], self=@n51, NewScrollBarVisibility=Visible)
   Event On Activated()
   > SetFrontEndCamera!(
    execute
    self:Fort Front End Context Object Reference=Get FortFrontEndContext!() -> (ReturnValue:Fort Front End Context Object Reference) {is=title K2Node_GetBlueprintContextEx}
    NewState:EFrontEndCamera Enum=Cosmetics
    TransitionParams:View Target Transition Params Structure=(BlendTime=0,BlendFunction=VTBlend_Cubic,BlendExp=2,bLockOutgoing=False)
   ) -> (then)
   n52 = "Get (a ref)"(Array=Get ItemDefs(), Dimension 1=0)
   > n53 = ConstructAnalyticsAttribute!(
    execute
    AttributeName:String=Item
    StringValue:String=GetDisplayName(Object=Break Fort Received Item Loot Info(FortReceivedItemLootInfo=@n52).ItemDef)
    FGuidValue:Guid Structure
    IntValue:int=-1
    floatValue:float=0
    BoolValue:bool=false
   ) -> (then, ReturnValue:Fort Analytics Event Attribute Structure)
   > n54 = ConstructAnalyticsAttribute(
    AttributeName=GiftBoxTemplateID
    StringValue=Get TemplateId!(self:Fort Persistable Item Object Reference=Get GiftBoxItem()) -> (TemplateId:String)
    IntValue=-1
    floatValue=0
    BoolValue=false
   )
   > FireAnalyticsEvent_BlueprintArray!(
    execute
    EventName:String=ItemReceivedScreenBaseSeen
    Attributes:Array of Fort Analytics Event Attribute Structures=Make Array!("[0]":Fort Analytics Event Attribute Structure=@n54, "[1]":Fort Analytics Event Attribute Structure=@n53) -> (Array:Array of Fort Analytics Event Attribute Structures)
   ) -> (then)
   n55 = Event OnInputModeChanged!() -> (OutputDelegate, then, bUsingGamepad:bool) {is=title K2Node_Event}
   > Branch(Condition=@n55.bUsingGamepad)
   > SelectItem(Index=0, CenterWidget=true)
   n56 = Event OnGiftBoxItemSet!() -> (OutputDelegate, then, bFromSelf:bool) {is=title K2Node_Event}
   > Set FromSelf(FromSelf:bool=@n56.bFromSelf) -> (Output_Get:bool)
   n57 = Length(TargetArray=Get ItemDefs())
   n58 = "integer > integer"(A=@n57, B=1)
   > Set Text!(
    execute
    self:Icon Text Button Object Reference=Get CloseButton()
    ButtonText:Text=Select(Option 0:Text="LOCTEXT(\"Collect and Close\")", Option 1:Text="LOCTEXT(\"Collect All and Close\")", Index=@n58) -> (ReturnValue:Text)
   ) -> (then) {is=title K2Node_CallFunction}
   > n59 = "SetText (Text)"(
    self=Get TextBlock_0() -> (TextBlock_0:Text Object Reference)
    InText=Select(
     Option 0="LOCTEXT(\"1 Item Collected!\")"
     Option 1=Format Text!(Format="LOCTEXT(\"{n} Items Collected!\")", n:int=@n57) -> (Result)
     Index=@n58
    )
   )
   n60 = Branch(Condition="Not Equal (Enum)"!(
     A:ESlateVisibility Enum=GetVisibility!(self:Widget Object Reference=@n51) -> (ReturnValue:ESlateVisibility Enum=Visible)
     B:ESlateVisibility Enum=Visible
    ) -> (ReturnValue:bool))
   SetInputActionHandler!(
    execute
    self:Common Activatable Panel Legacy Object Reference
    InputActionRow:Data Table Row Handle Structure=Make Data Table Row Handle!(DataTable:Data Table Object Reference=&InputActionDataTable, RowName:Name=Back) -> (DataTableRowHandle:Data Table Row Handle Structure) {is=title K2Node_MakeStruct}
    CommitedEvent:Delegate=Create Event()
   ) -> (then)
   Event Construct()
   Get RolloutTime()
  ]}
 GetWidgetForFramingViewedItem {nodes [
   GetWidgetForFramingViewedItem() {is=title K2Node_FunctionEntry}
   > Return Node(ReturnValue:Widget Object Reference=Get FramingWidget() -> (FramingWidget:Size Box Object Reference))
  ]}
 HandleBack {nodes [
   HandleBack() {is=title K2Node_FunctionEntry}
   > PopContentWidget!(execute) -> (then)
   > NavigateForGift!(execute, Item:Fort Gift Box Item Object Reference=Get GiftBoxItem()) -> (then)
   > Call On Final Gifting Complete!(execute:Exec, self:Item Received Base Object Reference) -> (then:Exec) {is=title K2Node_CallDelegate}
   > Return Node!(execute, Passthrough:bool=false)
  ]}
 HandleLeft {nodes [
   n1 = HandleLeft!() -> (then, Navigation Enum:EUINavigation Enum) {is=title K2Node_FunctionEntry}
   > n2 = HandleSelectionIncrement!(execute, self, Increment:bool=false) -> (then, Index:int=0)
   > PrintString!(execute, InString=Left, bPrintToScreen, bPrintToLog, TextColor, Duration, Key) -> (then)
   n3 = Get GiftScrollBox()
   > Return Node!(
    execute
    NewParam:Widget Object Reference=Select!(
     Left:Widget Object Reference=GetChildAt(self=@n3, Index=@n2)
     Right:Widget Object Reference
     Up:Widget Object Reference
     Down:Widget Object Reference
     Next:Widget Object Reference
     Previous:Widget Object Reference
     Invalid:Widget Object Reference
     Index:EUINavigation Enum=@n1
    ) -> (ReturnValue:Widget Object Reference)
   )
  ]}
 HandleRight {nodes [
   n1 = HandleRight!() -> (then, Nav Enum:EUINavigation Enum) {is=title K2Node_FunctionEntry}
   > n2 = HandleSelectionIncrement(Increment=true) -> (Index=0)
   > PrintString(InString=Right)
   n3 = Get GiftScrollBox()
   > Return Node(NewParam=Select(Right=GetChildAt(self=@n3, Index=@n2), Index=@n1))
  ]}
 HandleSelectionIncrement {nodes [
   n1 = HandleSelectionIncrement!() -> (then, Increment:bool=~) {is=title K2Node_FunctionEntry}
   > n2 = Sequence()
   > n3 = Branch(Condition=@n1)
   > Increment Int(Value=Get ItemSelectionIndex() -> (ItemSelectionIndex:int=0))
   n4 = GetChildrenCount(self=Get GiftScrollBox()) -> (ReturnValue=0)
   > Branch(Condition="integer >= integer"(A=Get ItemSelectionIndex() -> (ItemSelectionIndex=0), B=@n4))
   > Set ItemSelectionIndex(ItemSelectionIndex:int=int - int(A=@n4, B=1)) -> (Output_Get:int=0)
   Decrement Int(" "=@n3.else, Value=Get ItemSelectionIndex() -> (ItemSelectionIndex=0))
   > Branch(Condition="integer < integer"!(A:int=Get ItemSelectionIndex() -> (ItemSelectionIndex=0), B:int=0) -> (ReturnValue:bool) {is=title K2Node_CallFunction})
   > Set ItemSelectionIndex(ItemSelectionIndex=0) -> (Output_Get=0)
   Return Node!(execute=@n2.then_1, Index:int=Get ItemSelectionIndex() -> (ItemSelectionIndex=0))
  ]}
 UpdateFromHoverState {nodes [
   n1 = UpdateFromHoverState!() -> (then, Loot:Fort Received Item Loot Info Structure) {is=title K2Node_FunctionEntry}
   > UpdateFromItemReceivedInfo(Loot=@n1, select=true)
   > PopulateUsingItem(self=Get ItemInfoWidget(), NewItem=Break Fort Received Item Loot Info(FortReceivedItemLootInfo=@n1).GeneratedItemInstance)
  ]}
 UpdateFromItemReceivedInfo {nodes [
   n1 = UpdateFromItemReceivedInfo!() -> (then, Loot:Fort Received Item Loot Info Structure, select:bool=~) {is=title K2Node_FunctionEntry}
   n2 = Break Fort Received Item Loot Info(FortReceivedItemLootInfo=@n1)
   n3 = "SetText (Text)"(InText=Get ItemName!(self:Item Definition Base Object Reference=@n2.ItemDef) -> (ItemName:Text))
   Branch(execute=[@n3.then, @n1.then], Condition=@n1.select)
   > SetFrontEndCamera(
    self=Get FortFrontEndContext()
    NewState=Cosmetics
    TransitionParams=(BlendTime=0,BlendFunction=VTBlend_Cubic,BlendExp=2,bLockOutgoing=False)
   )
   n4 = GetItemDefinition!(self:Fort Item Object Reference=@n2.GeneratedItemInstance) -> (ReturnValue:Fort Item Definition Object Reference)
   > ViewVaultItem!(
    execute
    self:Fort Front End Context Object Reference=Get FortFrontEndContext()
    ItemToView:Fort Item Object Reference=@n2.GeneratedItemInstance
    bIgnoreCurrentlyEquippedFavorites:bool=Cast To AthenaCharacterItemDefinition!(Object=@n4) -> (AsAthena Character Item Definition:Athena Character Item Definition Object Reference, bSuccess:bool=~) {is=title K2Node_DynamicCast}.bSuccess
    DisplayContext:EFortPawnDisplayContext Enum=Select!(
     Option 0:EFortPawnDisplayContext Enum=VaultItem
     Option 1:EFortPawnDisplayContext Enum=VaultItemFromDefinition
     Index:bool=Cast To FortHero!(Object=@n2.GeneratedItemInstance) -> (AsFort Hero:Fort Hero Object Reference, bSuccess:bool=~) {is=title K2Node_DynamicCast}.bSuccess
    ) -> (ReturnValue:EFortPawnDisplayContext Enum)
   ) -> (then)
   n5 = GetItemDefinitionOwnershipState!(
    self:Fort Player Controller Object Reference=GetFirstLocalFortPlayerController!(bIncludeNonSpectators:bool=true, bIncludeSpectators:bool=false) -> (ReturnValue:Fort Player Controller Object Reference)
    ItemDefinition:Fort Item Definition Object Reference=@n4
   ) -> (bIsOwned:bool, ReturnValue:bool)
   n6 = Get GiftBoxItem()
   n7 = HasTag!(
    TagContainer:"Gameplay Tag Container Structure (by ref)"=Get Owned Gameplay Tags!(self:Gameplay Tag Asset Interface Interface=GetGiftBoxDefinition(self=@n6)) -> (ReturnValue:Gameplay Tag Container Structure) {is=title K2Node_CallFunction}
    Tag:Gameplay Tag Structure=(TagName="Item.Consumable.UnGiftBox")
    bExactMatch:bool=true
   ) -> (ReturnValue:bool)
   n8 = Cast To AthenaCosmeticItemDefinition!(Object=@n4) -> (AsAthena Cosmetic Item Definition:Athena Cosmetic Item Definition Object Reference, bSuccess:bool=~) {is=title K2Node_DynamicCast}
   n9 = Cast To FortVariantTokenType!(Object=@n4) -> (AsFort Variant Token Type:Fort Variant Token Type Object Reference, bSuccess:bool=~) {is=title K2Node_DynamicCast}
   > n10 = Branch(Condition=AND Boolean!(A=NOT Boolean(A=@n5), B=@n5.ReturnValue, C:bool=OR Boolean(A=@n8.bSuccess, B=@n9.bSuccess), D:bool=NOT Boolean(A=@n7)) -> (ReturnValue))
   n11 = Get RichText_UnownedBaseItemMessage() -> (RichText_UnownedBaseItemMessage:Common Rich Text Block Object Reference)
   > SetVisibility(self=@n11, InVisibility=Visible)
   SetVisibility(execute=@n10.else, self=@n11, InVisibility=Collapsed)
  ]}
]
variables [
 AnimationIndex {type=integer}
 AnimationIndexCached {type=integer}
 CurrentIndex {type=integer}
 FromSelf {type=boolean}
 ItemSelectionIndex {type=integer}
 RolloutTime {type=number, default=0.25}
 SoundOnItemReceived {type=SoundBase, default=/Game/Sounds/Fort_UI_Sounds/FrontEnd/RewardsAndGifts/Gifting_ItemReceived_Cue}
 UseClickRollout {type=boolean, default=true}
]
widgets {
 info {inheritedWidgetCount=2, parentClass=/Script/FortniteUI.FortItemReceivedScreen, rootWidgetClass=&SafeZone, widgetCount=26}
 widgets [
  MainSafeZone {widgetClassPath=&SafeZone, widget=MainSafeZone, properties {padBottom=false, padTop=false}}
  Overlay_2 {widgetClassPath=&Overlay, parent=MainSafeZone, widget=Overlay_2, slot {class=SafeZoneSlot}}
  HorizontalBox_2 {
   widgetClassPath=&HorizontalBox
   parent=Overlay_2
   widget=HorizontalBox_2
   slot {class=OverlaySlot, properties {horizontalAlignment=HAlign_Fill, verticalAlignment=VAlign_Fill}}
  }
  SizeBox_Contents {
   widgetClassPath=&SizeBox
   bIsVariable=true
   parent=HorizontalBox_2
   widget=SizeBox_Contents
   properties {bOverride_WidthOverride=true, maxDesiredWidth=2568.8411, widthOverride=1065}
   slot {class=HorizontalBoxSlot, properties {horizontalAlignment=HAlign_Left, padding=<ltrb 50 90 0 0>}}
  }
  LeftContainer {widgetClassPath=&VerticalBox, parent=SizeBox_Contents, widget=LeftContainer, slot {class=SizeBoxSlot}}
  HeaderContainer {
   widgetClassPath=&HorizontalBox
   bInherited=true
   parent=LeftContainer
   widget=HeaderContainer
   slot {class=VerticalBoxSlot, properties {horizontalAlignment=HAlign_Center}}
  }
  SizeBox_1 {
   widgetClassPath=&SizeBox
   parent=LeftContainer
   widget=SizeBox_1
   properties {bOverride_WidthOverride=true, clipping=ClipToBounds, maxDesiredWidth=2375.1282, widthOverride=821.4651}
   slot {class=VerticalBoxSlot, properties {padding=<ltrb 0 -200 0 -280>, size {sizeRule=Fill, value=1}}}
  }
  Overlay_0 {widgetClassPath=&Overlay, parent=SizeBox_1, widget=Overlay_0, slot {class=SizeBoxSlot}}
  Image_DarkBackground {
   widgetClassPath=&Image
   bIsVariable=true
   parent=Overlay_0
   widget=Image_DarkBackground
   properties {visibility=HitTestInvisible, colorAndOpacity=<rgba 0 0 0 0.1>}
   slot {class=OverlaySlot, properties {horizontalAlignment=HAlign_Fill, verticalAlignment=VAlign_Fill}}
  }
  BorderRarityFlare {
   widgetClassPath=&Border
   parent=Overlay_0
   widget=BorderRarityFlare
   properties {
    background {
     bIsDynamicallyLoaded=false
     drawAs=NoDrawType
     imageType=NoImage
     mirroring=NoMirror
     resourceName=None
     resourceObject=null
     tiling=NoTile
     imageSize=<32 32>
     margin=<ltrb 0 0 0 0>
     outlineSettings {
      bUseBrushTransparency=false
      roundingType=HalfHeightRadius
      width=0
      color {colorUseRule=UseColor_Specified, specifiedColor=<rgba 0 0 0 0>}
      cornerRadii=<0 0 0 0>
     }
     tintColor {colorUseRule=UseColor_Specified, specifiedColor=<rgba 1 1 1 1>}
     uVRegion {bIsValid=false, max=<0 0>, min=<0 0>}
    }
    padding=<ltrb 0 0 0 0>
   }
   slot {class=OverlaySlot, properties {horizontalAlignment=HAlign_Center, verticalAlignment=VAlign_Center}}
  }
  ImageRarityFlare {
   widgetClassPath=&Image
   bIsVariable=true
   parent=BorderRarityFlare
   widget=ImageRarityFlare
   properties {
    visibility=Hidden
    brush {
     bIsDynamicallyLoaded=false
     drawAs=Image
     imageType=NoImage
     mirroring=NoMirror
     resourceName=None
     resourceObject=/Game/UI/Frontend/Rewards/materials/M_UI_DailyReward_Rarity-Flares
     tiling=NoTile
     imageSize=<256 256>
     margin=<ltrb 0 0 0 0>
     outlineSettings {
      bUseBrushTransparency=false
      roundingType=HalfHeightRadius
      width=0
      color {colorUseRule=UseColor_Specified, specifiedColor=<rgba 0 0 0 0>}
      cornerRadii=<0 0 0 0>
     }
     tintColor {colorUseRule=UseColor_Specified, specifiedColor=<rgba 1 1 1 1>}
     uVRegion {bIsValid=false, max=<0 0>, min=<0 0>}
    }
    renderTransform {angle=0, scale=<2.75 1.875>, shear=<0 0>, translation=<0 0>}
   }
   slot {class=BorderSlot, properties {padding=<ltrb 0 0 0 0>}}
  }
  GiftScrollBox {
   widgetClassPath=&ScrollBox
   bInherited=true
   bIsVariable=true
   parent=Overlay_0
   widget=GiftScrollBox
   properties {
    allowOverscroll=false
    clipping=Inherit
    consumeMouseWheel=Always
    navigation=&self:WidgetTree.GiftScrollBox.WidgetNavigation_0
    navigationScrollPadding=16
    orientation=Orient_Horizontal
    widgetStyle {
     barThickness=9
     bottomShadowBrush {
      bIsDynamicallyLoaded=false
      drawAs=Box
      imageType=FullColor
      mirroring=NoMirror
      resourceName=../../../Engine/Content/Slate/Common/ScrollBoxShadowBottom.png
      resourceObject=null
      tiling=NoTile
      imageSize=<16 8>
      margin=<ltrb 0.5 0 0.5 1>
      outlineSettings {
       bUseBrushTransparency=false
       roundingType=HalfHeightRadius
       width=0
       color {colorUseRule=UseColor_Specified, specifiedColor=<rgba 0 0 0 0>}
       cornerRadii=<0 0 0 0>
      }
      tintColor {colorUseRule=UseColor_Specified, specifiedColor=<rgba 1 1 1 1>}
      uVRegion {bIsValid=false, max=<0 0>, min=<0 0>}
     }
     horizontalScrolledContentPadding=<ltrb 0 0 1 0>
     leftShadowBrush {
      bIsDynamicallyLoaded=false
      drawAs=Box
      imageType=FullColor
      mirroring=NoMirror
      resourceName=../../../Engine/Content/Slate/Common/ScrollBoxShadowLeft.png
      resourceObject=null
      tiling=NoTile
      imageSize=<0 0>
      margin=<ltrb 1 0.5 0 0.5>
      outlineSettings {
       bUseBrushTransparency=false
       roundingType=HalfHeightRadius
       width=0
       color {colorUseRule=UseColor_Specified, specifiedColor=<rgba 0 0 0 0>}
       cornerRadii=<0 0 0 0>
      }
      tintColor {colorUseRule=UseColor_Specified, specifiedColor=<rgba 1 1 1 1>}
      uVRegion {bIsValid=false, max=<0 0>, min=<0 0>}
     }
     rightShadowBrush {
      bIsDynamicallyLoaded=false
      drawAs=Box
      imageType=FullColor
      mirroring=NoMirror
      resourceName=../../../Engine/Content/Slate/Common/ScrollBoxShadowRight.png
      resourceObject=null
      tiling=NoTile
      imageSize=<0 0>
      margin=<ltrb 0 0.5 1 0.5>
      outlineSettings {
       bUseBrushTransparency=false
       roundingType=HalfHeightRadius
       width=0
       color {colorUseRule=UseColor_Specified, specifiedColor=<rgba 0 0 0 0>}
       cornerRadii=<0 0 0 0>
      }
      tintColor {colorUseRule=UseColor_Specified, specifiedColor=<rgba 1 1 1 1>}
      uVRegion {bIsValid=false, max=<0 0>, min=<0 0>}
     }
     topShadowBrush {
      bIsDynamicallyLoaded=false
      drawAs=NoDrawType
      imageType=FullColor
      mirroring=NoMirror
      resourceName=../../../Engine/Content/Slate/Common/ScrollBoxShadowTop.png
      resourceObject=null
      tiling=NoTile
      imageSize=<0 0>
      margin=<ltrb 0.5 1 0.5 0>
      outlineSettings {
       bUseBrushTransparency=false
       roundingType=HalfHeightRadius
       width=0
       color {colorUseRule=UseColor_Specified, specifiedColor=<rgba 0 0 0 0>}
       cornerRadii=<0 0 0 0>
      }
      tintColor {colorUseRule=UseColor_Specified, specifiedColor=<rgba 1 1 1 1>}
      uVRegion {bIsValid=false, max=<0 0>, min=<0 0>}
     }
     verticalScrolledContentPadding=<ltrb 0 0 0 1>
    }
   }
   slot {class=OverlaySlot, properties {horizontalAlignment=HAlign_Center, verticalAlignment=VAlign_Center}}
  }
  VerticalBox_0 {
   widgetClassPath=&VerticalBox
   parent=LeftContainer
   widget=VerticalBox_0
   slot {class=VerticalBoxSlot, properties {horizontalAlignment=HAlign_Center, verticalAlignment=VAlign_Bottom, padding=<ltrb 0 25 0 80>}}
  }
  RichText_UnownedBaseItemMessage {
   widgetClassPath=&CommonRichTextBlock
   bIsVariable=true
   parent=VerticalBox_0
   widget=RichText_UnownedBaseItemMessage
   properties {
    autoWrapText=true
    bOverrideDefaultStyle=true
    justification=Center
    text=You unlocked a style for an item that you do not yet own.
    visibility=Collapsed
    defaultTextStyleOverride {
     overflowPolicy=Clip
     transformPolicy=None
     colorAndOpacity {colorUseRule=UseColor_Specified, specifiedColor=<rgba 0.2831 0.6376 0.9823 1>}
     font {
      bForceMonospaced=false
      bMaterialIsStencil=false
      fontMaterial=null
      fontObject=&BurbankSmall
      letterSpacing=0
      monospacedWidth=1
      size=18
      skewAmount=0
      typefaceFontName=Bold
      outlineSettings {
       bApplyOutlineToDropShadows=false
       bMiteredCorners=false
       bSeparateFillAlpha=false
       outlineMaterial=null
       outlineSize=0
       outlineColor=<rgba 0 0 0 1>
      }
     }
     highlightColor {colorUseRule=UseColor_Specified, specifiedColor=<rgba 0 0 0 1>}
     highlightShape {
      bIsDynamicallyLoaded=false
      drawAs=Image
      imageType=NoImage
      mirroring=NoMirror
      resourceName=None
      resourceObject=null
      tiling=NoTile
      imageSize=<32 32>
      margin=<ltrb 0 0 0 0>
      outlineSettings {
       bUseBrushTransparency=false
       roundingType=HalfHeightRadius
       width=0
       color {colorUseRule=UseColor_Specified, specifiedColor=<rgba 0 0 0 0>}
       cornerRadii=<0 0 0 0>
      }
      tintColor {colorUseRule=UseColor_Specified, specifiedColor=<rgba 1 1 1 1>}
      uVRegion {bIsValid=false, max=<0 0>, min=<0 0>}
     }
     selectedBackgroundColor {colorUseRule=UseColor_Foreground, specifiedColor=<rgba 1 0 1 1>}
     shadowColorAndOpacity=<rgba 0 0 0 1>
     shadowOffset=<0 0>
     strikeBrush {
      bIsDynamicallyLoaded=false
      drawAs=Image
      imageType=NoImage
      mirroring=NoMirror
      resourceName=None
      resourceObject=null
      tiling=NoTile
      imageSize=<32 32>
      margin=<ltrb 0 0 0 0>
      outlineSettings {
       bUseBrushTransparency=false
       roundingType=HalfHeightRadius
       width=0
       color {colorUseRule=UseColor_Specified, specifiedColor=<rgba 0 0 0 0>}
       cornerRadii=<0 0 0 0>
      }
      tintColor {colorUseRule=UseColor_Specified, specifiedColor=<rgba 1 1 1 1>}
      uVRegion {bIsValid=false, max=<0 0>, min=<0 0>}
     }
     underlineBrush {
      bIsDynamicallyLoaded=false
      drawAs=Image
      imageType=NoImage
      mirroring=NoMirror
      resourceName=None
      resourceObject=null
      tiling=NoTile
      imageSize=<32 32>
      margin=<ltrb 0 0 0 0>
      outlineSettings {
       bUseBrushTransparency=false
       roundingType=HalfHeightRadius
       width=0
       color {colorUseRule=UseColor_Specified, specifiedColor=<rgba 0 0 0 0>}
       cornerRadii=<0 0 0 0>
      }
      tintColor {colorUseRule=UseColor_Specified, specifiedColor=<rgba 1 1 1 1>}
      uVRegion {bIsValid=false, max=<0 0>, min=<0 0>}
     }
    }
    renderTransform {angle=0, scale=<1 1>, shear=<-8 0>, translation=<0 0>}
   }
   slot {class=VerticalBoxSlot, properties {padding=<ltrb 0 0 0 12>}}
  }
  Overlay_4 {
   widgetClassPath=&Overlay
   parent=VerticalBox_0
   widget=Overlay_4
   properties {visibility=Collapsed}
   slot {class=VerticalBoxSlot, properties {horizontalAlignment=HAlign_Center}}
  }
  AthenaGiftboxUnownedMessage {
   widgetClassPath=/Game/Athena/UI/Customization/AthenaGiftboxUnownedMessage
   bIsVariable=true
   parent=Overlay_4
   widget=AthenaGiftboxUnownedMessage
   properties {visibility=Collapsed}
   slot {class=OverlaySlot}
  }
  Text_ItemCount {
   widgetClassPath=&CommonTextBlock
   bIsVariable=true
   parent=VerticalBox_0
   widget=Text_ItemCount
   properties {
    style=&TextStyle-Base_C
    text=0/10
    visibility=Hidden
    font {
     bForceMonospaced=false
     bMaterialIsStencil=false
     fontMaterial=null
     fontObject=&NotoSans
     letterSpacing=0
     monospacedWidth=1
     size=24
     skewAmount=0
     typefaceFontName=NotoSans-Regular
     outlineSettings {
      bApplyOutlineToDropShadows=false
      bMiteredCorners=false
      bSeparateFillAlpha=false
      outlineMaterial=null
      outlineSize=0
      outlineColor=<rgba 0 0 0 1>
     }
    }
    shadowOffset=<0 0>
   }
   slot {class=VerticalBoxSlot, properties {horizontalAlignment=HAlign_Center}}
  }
  Overlay_3 {widgetClassPath=&Overlay, parent=VerticalBox_0, widget=Overlay_3, slot {class=VerticalBoxSlot}}
  HorizontalBoxActions {
   widgetClassPath=&HorizontalBox
   parent=Overlay_3
   widget=HorizontalBoxActions
   slot {class=OverlaySlot, properties {horizontalAlignment=HAlign_Center}}
  }
  ShowNextButton {
   widgetClassPath=&IconTextButton
   bIsVariable=true
   parent=HorizontalBoxActions
   widget=ShowNextButton
   properties {
    bDisplayInputActionWhenNotInteractable=false
    buttonText=Show Next
    contentAlignment=HAlign_Center
    minHeight=70
    minWidth=300
    navigation=&self:WidgetTree.ShowNextButton.WidgetNavigation_0
    onButtonBaseClicked=(&self:WidgetTree.ShowNextButton.HandleOnButtonClicked)
    onButtonBaseDoubleClicked=(&self:WidgetTree.ShowNextButton.HandleOnButtonDoubleClicked)
    onButtonBaseHovered=(&self:WidgetTree.ShowNextButton.HandleOnButtonHovered)
    onButtonBaseUnhovered=(&self:WidgetTree.ShowNextButton.HandleOnButtonUnhovered)
    onSelectedChangedBase=(&self:WidgetTree.ShowNextButton.HandleOnSelectedChanged)
    style=/FortUILibrary/Foundation/Buttons/ButtonStyles/TopBar/ButtonStyle-Outline-Green.ButtonStyle-Outline-Green_C
    triggeringInputAction {dataTable=&InputActionDataTable, rowName=GiftScreenNextItem}
   }
   slot {class=HorizontalBoxSlot, properties {horizontalAlignment=HAlign_Center, verticalAlignment=VAlign_Center, padding=<ltrb 12 12 12 12>}}
  }
  EquipButton {
   widgetClassPath=&IconTextButton
   bIsVariable=true
   parent=HorizontalBoxActions
   widget=EquipButton
   properties {
    bDisplayInputActionWhenNotInteractable=false
    buttonText=Equip
    contentAlignment=HAlign_Center
    minHeight=70
    minWidth=300
    navigation=&self:WidgetTree.EquipButton.WidgetNavigation_0
    onButtonBaseClicked=(&self:WidgetTree.EquipButton.HandleOnButtonClicked)
    onButtonBaseDoubleClicked=(&self:WidgetTree.EquipButton.HandleOnButtonDoubleClicked)
    onButtonBaseHovered=(&self:WidgetTree.EquipButton.HandleOnButtonHovered)
    onButtonBaseUnhovered=(&self:WidgetTree.EquipButton.HandleOnButtonUnhovered)
    onSelectedChangedBase=(&self:WidgetTree.EquipButton.HandleOnSelectedChanged)
    style=/FortUILibrary/Foundation/Buttons/ButtonStyles/TopBar/ButtonStyle-Outline.ButtonStyle-Outline_C
    visibility=Collapsed
    triggeringInputAction {dataTable=&InputActionDataTable, rowName=Equip}
   }
   slot {class=HorizontalBoxSlot, properties {horizontalAlignment=HAlign_Center, verticalAlignment=VAlign_Center, padding=<ltrb 12 12 12 12>}}
  }
  CloseButton {
   widgetClassPath=&IconTextButton
   bIsVariable=true
   parent=HorizontalBoxActions
   widget=CloseButton
   properties {
    bDisplayInputActionWhenNotInteractable=false
    buttonText=Collect and Close
    contentAlignment=HAlign_Center
    minHeight=70
    minWidth=300
    navigation=&self:WidgetTree.CloseButton.WidgetNavigation_0
    onButtonBaseClicked=(&self:WidgetTree.CloseButton.HandleOnButtonClicked)
    onButtonBaseDoubleClicked=(&self:WidgetTree.CloseButton.HandleOnButtonDoubleClicked)
    onButtonBaseHovered=(&self:WidgetTree.CloseButton.HandleOnButtonHovered)
    onButtonBaseUnhovered=(&self:WidgetTree.CloseButton.HandleOnButtonUnhovered)
    onSelectedChangedBase=(&self:WidgetTree.CloseButton.HandleOnSelectedChanged)
    style=&ButtonStyle-Outline_C
    visibility=Visible
    triggeringInputAction {dataTable=&InputActionDataTable, rowName=CollectAll}
   }
   slot {class=HorizontalBoxSlot, properties {horizontalAlignment=HAlign_Center, verticalAlignment=VAlign_Center, padding=<ltrb 12 12 12 12>}}
  }
  TextBlock_0 {
   widgetClassPath=&TextBlock
   bIsVariable=true
   parent=Overlay_3
   widget=TextBlock_0
   properties {
    text=X Items Collected
    visibility=Hidden
    font {
     bForceMonospaced=false
     bMaterialIsStencil=false
     fontMaterial=null
     fontObject=&BurbankSmall
     letterSpacing=0
     monospacedWidth=1
     size=24
     skewAmount=0
     typefaceFontName=Bold
     outlineSettings {
      bApplyOutlineToDropShadows=false
      bMiteredCorners=false
      bSeparateFillAlpha=false
      outlineMaterial=null
      outlineSize=0
      outlineColor=<rgba 0 0 0 1>
     }
    }
   }
   slot {class=OverlaySlot, properties {horizontalAlignment=HAlign_Center, verticalAlignment=VAlign_Fill, padding=<ltrb 30 30 30 30>}}
  }
  Overlay_1 {
   widgetClassPath=&Overlay
   parent=HorizontalBox_2
   widget=Overlay_1
   slot {class=HorizontalBoxSlot, properties {size {sizeRule=Fill, value=1}}}
  }
  FramingWidget {
   widgetClassPath=&SizeBox
   bIsVariable=true
   parent=Overlay_1
   widget=FramingWidget
   slot {class=OverlaySlot, properties {horizontalAlignment=HAlign_Fill, verticalAlignment=VAlign_Fill}}
  }
  ItemInfoWidget {
   widgetClassPath=/Game/Athena/UI/Season/ItemInfoWidget
   bIsVariable=true
   parent=Overlay_1
   widget=ItemInfoWidget
   properties {visibility=Visible}
   slot {class=OverlaySlot, properties {horizontalAlignment=HAlign_Right, verticalAlignment=VAlign_Bottom, padding=<ltrb 0 0 30 30>}}
  }
 ]
}
