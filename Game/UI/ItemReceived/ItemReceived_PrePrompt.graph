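# Fortnite gift pre-prompt screen that shows a received gift box, its sender and an optional gift tag before opening.
# Setting the gift item updates title, sender and box image, and profanity-filters any user message for the tag.
# Activation plays intro and looping idle animations with sound, reveals Open after a delay and logs analytics.
# Open pops this panel and pushes ItemReceived_Base, whose final-gifting-complete event fires On Pre Prompt Closed.
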WidgetBlueprint /Game/UI/ItemReceived/ItemReceived_PrePrompt
animations [
 Init {
  binding_count=8
  display_label=""
  length_seconds=1.86
  object=&self:Init
  sequence {
   bindings [
    UserMessageBorder {
     id=#1
     display_name=Overlay_GiftTag
     possessed_class=CommonBorder
     tracks [
      {
       type=MovieSceneFloatTrack
       display_name=Render Opacity
       sections [{type=MovieSceneFloatSection, channels [None {type=MovieSceneScriptingFloatChannel, keys [[116.505,0],[137.9217,1]]}]}]
      }
      {
       type=MovieScene2DTransformTrack
       sections [{
         type=MovieScene2DTransformSection
         channels [
          Translation.X {type=MovieSceneScriptingFloatChannel}
          Translation.Y {type=MovieSceneScriptingFloatChannel}
          Angle {type=MovieSceneScriptingFloatChannel}
          Scale.X {type=MovieSceneScriptingFloatChannel, keys [[112.85,0],[134.355,1.0892],[138.025,0.9826],[143.515,1]]}
          Scale.Y {type=MovieSceneScriptingFloatChannel, keys [[112.85,0],[133.495,1.1864],[137.3667,0.9397],[143.515,1]]}
          Shear.X {type=MovieSceneScriptingFloatChannel}
          Shear.Y {type=MovieSceneScriptingFloatChannel}
         ]
        }]
      }
     ]
    }
    VerticalBox_0 {
     id=#2
     possessed_class=VerticalBox
     tracks [
      {
       type=MovieSceneFloatTrack
       display_name=Render Opacity
       sections [{type=MovieSceneFloatSection, channels [None {type=MovieSceneScriptingFloatChannel, keys [[11.9983,0],[28.9983,1]]}]}]
      }
      {
       type=MovieScene2DTransformTrack
       sections [{
         type=MovieScene2DTransformSection
         channels [
          Translation.X {type=MovieSceneScriptingFloatChannel}
          Translation.Y {type=MovieSceneScriptingFloatChannel}
          Angle {type=MovieSceneScriptingFloatChannel}
          Scale.X {type=MovieSceneScriptingFloatChannel, keys [[10.9983,0],[25,1.2],[45.9983,1]]}
          Scale.Y {type=MovieSceneScriptingFloatChannel, keys [[10.9983,0],[25,1.2],[42,1]]}
          Shear.X {type=MovieSceneScriptingFloatChannel}
          Shear.Y {type=MovieSceneScriptingFloatChannel}
         ]
        }]
      }
     ]
    }
    BackgroundBlur_0 {
     id=#3
     possessed_class=BackgroundBlur
     tracks [{
       type=MovieSceneFloatTrack
       display_name=Blur Strength
       sections [{type=MovieSceneFloatSection, end_frame=40, start_frame=0, channels [None {type=MovieSceneScriptingFloatChannel, keys [[0,0],[20,25]]}]}]
      }]
    }
    ItemReceived_PrePrompt_C_32 {
     id=#4
     possessed_class=ItemReceived_PrePrompt_C
     tracks [{
       type=MovieSceneColorTrack
       display_name=Color and Opacity
       sections [{
         type=MovieSceneColorSection
         end_frame=100
         start_frame=0
         channels [
          Color.R {type=MovieSceneScriptingFloatChannel}
          Color.G {type=MovieSceneScriptingFloatChannel}
          Color.B {type=MovieSceneScriptingFloatChannel}
          Color.A {type=MovieSceneScriptingFloatChannel, keys [[0,0],[40,1]]}
         ]
        }]
      }]
    }
    OpenButton {
     id=#5
     possessed_class=IconTextButton_C
     tracks [
      {
       type=MovieSceneFloatTrack
       display_name=Render Opacity
       sections [{type=MovieSceneFloatSection, channels [None {type=MovieSceneScriptingFloatChannel, keys [[124,0],[137,1]]}]}]
      }
      {
       type=MovieScene2DTransformTrack
       sections [{
         type=MovieScene2DTransformSection
         channels [
          Translation.X {type=MovieSceneScriptingFloatChannel}
          Translation.Y {type=MovieSceneScriptingFloatChannel, keys [[124,85.759],[137,0]]}
          Angle {type=MovieSceneScriptingFloatChannel}
          Scale.X {type=MovieSceneScriptingFloatChannel}
          Scale.Y {type=MovieSceneScriptingFloatChannel}
          Shear.X {type=MovieSceneScriptingFloatChannel}
          Shear.Y {type=MovieSceneScriptingFloatChannel}
         ]
        }]
      }
     ]
    }
    OverlayTriangleParticles {
     id=#6
     possessed_class=Overlay
     tracks [
      {
       type=MovieScene2DTransformTrack
       sections [{
         type=MovieScene2DTransformSection
         end_frame=120
         start_frame=0
         channels [
          Translation.X {type=MovieSceneScriptingFloatChannel}
          Translation.Y {type=MovieSceneScriptingFloatChannel}
          Angle {type=MovieSceneScriptingFloatChannel}
          Scale.X {type=MovieSceneScriptingFloatChannel, keys [[94.9983,0],[114.9983,0.8]]}
          Scale.Y {type=MovieSceneScriptingFloatChannel, keys [[94.9983,0],[114.9983,0.9]]}
          Shear.X {type=MovieSceneScriptingFloatChannel}
          Shear.Y {type=MovieSceneScriptingFloatChannel}
         ]
        }]
      }
      {
       type=MovieSceneFloatTrack
       display_name=Render Opacity
       sections [{type=MovieSceneFloatSection, channels [None {type=MovieSceneScriptingFloatChannel, keys [[165.9983,1],[185.9983,0.4]]}]}]
      }
     ]
    }
    SizeBoxGiftBox {
     id=#7
     display_name=SizeBox_0
     possessed_class=CommonLazyImage
     tracks [{
       type=MovieScene2DTransformTrack
       sections [{
         type=MovieScene2DTransformSection
         end_frame=141
         start_frame=0
         channels [
          Translation.X {type=MovieSceneScriptingFloatChannel, keys [[79.9983,0],[100,0.0491],[114.9983,0],[136,0]]}
          Translation.Y {type=MovieSceneScriptingFloatChannel, keys [[69.9983,0],[99.9983,-43.8508],[104.9983,0]]}
          Angle {type=MovieSceneScriptingFloatChannel, const [0,1]}
          Scale.X {type=MovieSceneScriptingFloatChannel, keys [[79.9983,0],[94.9983,1.2],[104.9983,1]]}
          Scale.Y {type=MovieSceneScriptingFloatChannel, keys [[79.9983,0],[94.9983,1.2],[105.9983,0.9111],[123.9983,1]]}
          Shear.X {type=MovieSceneScriptingFloatChannel, const [0,1]}
          Shear.Y {type=MovieSceneScriptingFloatChannel, const [0,1]}
         ]
        }]
      }]
    }
    BorderRarityFlare {
     id=#8
     possessed_class=Border
     tracks [{
       type=MovieScene2DTransformTrack
       sections [{
         type=MovieScene2DTransformSection
         end_frame=133
         start_frame=0
         channels [
          Translation.X {type=MovieSceneScriptingFloatChannel}
          Translation.Y {type=MovieSceneScriptingFloatChannel}
          Angle {type=MovieSceneScriptingFloatChannel}
          Scale.X {type=MovieSceneScriptingFloatChannel, keys [[88.9983,0],[101.9983,1.1],[127.9983,1]]}
          Scale.Y {type=MovieSceneScriptingFloatChannel, keys [[88.9983,0],[101.9983,1.1],[127.9983,1]]}
          Shear.X {type=MovieSceneScriptingFloatChannel}
          Shear.Y {type=MovieSceneScriptingFloatChannel}
         ]
        }]
      }]
    }
   ]
   display_rate {numerator=100}
  }
 }
 Idle {
  binding_count=2
  display_label=""
  length_seconds=2.1
  object=&self:Idle
  sequence {bindings [
    Image_GiftBox {
     id=#9
     display_name=CommonLazyImage_0
     possessed_class=Image
     tracks [{
       type=MovieScene2DTransformTrack
       sections [{
         type=MovieScene2DTransformSection
         end_frame=42
         start_frame=0
         channels [
          Translation.X {type=MovieSceneScriptingFloatChannel, const [0,1]}
          Translation.Y {type=MovieSceneScriptingFloatChannel, const [0,1]}
          Angle {type=MovieSceneScriptingFloatChannel, const [0,1]}
          Scale.X {type=MovieSceneScriptingFloatChannel, const [1,1]}
          Scale.Y {type=MovieSceneScriptingFloatChannel, const [1,1]}
          Shear.X {type=MovieSceneScriptingFloatChannel, const [0,1]}
          Shear.Y {type=MovieSceneScriptingFloatChannel, const [0,1]}
         ]
        }]
      }]
    }
    BorderRarityFlare {id=#10, possessed_class=Border, tracks []}
   ]}
 }
 Scale {binding_count=0, display_label="", length_seconds=5, object=&self:Scale, sequence {bindings []}}
 ScaleUP {
  binding_count=2
  display_label=""
  length_seconds=0.5
  object=&self:ScaleUP
  sequence {bindings [
    GiftBox {
     id=#9
     possessed_class=Image
     tracks [{
       type=MovieScene2DTransformTrack
       sections [{
         type=MovieScene2DTransformSection
         end_frame=10
         start_frame=0
         channels [
          Translation.X {type=MovieSceneScriptingFloatChannel}
          Translation.Y {type=MovieSceneScriptingFloatChannel}
          Angle {type=MovieSceneScriptingFloatChannel, keys [[0,0],[3.9997,8],[10,1]]}
          Scale.X {type=MovieSceneScriptingFloatChannel, keys [[0,0],[3.9997,1],[6,1.3],[9.9997,1]]}
          Scale.Y {type=MovieSceneScriptingFloatChannel, keys [[0,0],[3.9997,1],[6,1.3],[9.9997,1]]}
          Shear.X {type=MovieSceneScriptingFloatChannel, const [0,1]}
          Shear.Y {type=MovieSceneScriptingFloatChannel, const [0,1]}
         ]
        }]
      }]
    }
    BorderRarityFlare {id=#10, possessed_class=Border, tracks []}
   ]}
 }
]
graphs [
 EventGraph {nodes [
   "On Button Clicked (OpenButton)"!() -> (OutputDelegate, then, Button:Common Button Legacy Object Reference) {is=title K2Node_ComponentBoundEvent}
   > PopActivatablePanelInModalLayer!(execute, Panel:Common Activatable Panel Legacy Object Reference=Self-Reference()) -> (then)
   > n1 = Create Widget!(
    execute:Exec
    Class:User Widget Class Reference=&/ItemReceived_Base.ItemReceived_Base_C
    OwningPlayer:Player Controller Object Reference
   ) -> (then:Exec, ReturnValue:Item Received Base Object Reference)
   > SetGiftBoxItem!(
    execute
    self:Fort Item Received Widget Base Object Reference=@n1
    GiftBoxItem:Fort Gift Box Item Object Reference=Get GiftBoxItem() -> (GiftBoxItem:Fort Gift Box Item Object Reference)
   ) -> (then)
   > n2 = PushContentWidgetAdvanced!(execute, PushWidget:Widget Object Reference=@n1, bHideHeader:bool=true, bHideFooter:bool=true, bHideChatWidget:bool=true) -> (then)
   n3 = CustomEvent_0() {is=title K2Node_CustomEvent}
   Bind Event to On Final Gifting Complete(execute=@n2.then, self:Fort Item Received Screen Object Reference=@n1, Delegate=@n3) {is=title K2Node_AddDelegate}
   n4 = OnUserMsgFiltered!() -> (OutputDelegate, then, bSuccess:bool=~, SanitzedMsg:String) {is=title K2Node_CustomEvent}
   n5 = "To Text (String)"(InString=Get FromUsername() -> (FromUsername:String))
   n6 = Get fromSelf() -> (fromSelf:bool)
   > "SetText (Text)"(
    self=Get Text_UserMsg() -> (Text_UserMsg:Common Text Object Reference)
    InText=Format Text!(
     Format="LOCTEXT(\"{message}\\r\\n-{sender}\")"
     message:Text="To Text (String)"(InString=@n4.SanitzedMsg)
     sender:Text=Select(Option 0:Text=@n5, Option 1:Text="LOCTEXT(\"You\")", Index=@n6) -> (ReturnValue:Text)
    ) -> (Result)
   )
   Event On Activated()
   > PlayAnimation(InAnimation=Get Init() -> (Init:Widget Animation Object Reference))
   > PlayAnimation(InAnimation=Get Idle() -> (Idle:Widget Animation Object Reference), NumLoopsToPlay=0, PlayMode=PingPong)
   > PlayAnimation(InAnimation=Get ScaleUP() -> (ScaleUP:Widget Animation Object Reference))
   > PlaySound2D(Sound=/Game/Sounds/Fort_UI_Sounds/FrontEnd/RewardsAndGifts/Gifting_ItemRollout_Cue)
   > Delay(Duration=0.2)
   > SetVisibility(self=Get OpenButton() -> (OpenButton:Icon Text Button Object Reference), InVisibility=Visible)
   > n7 = ConstructAnalyticsAttribute!(
    execute
    AttributeName:String=FromUser
    StringValue:String=Get FromUsername()
    FGuidValue:Guid Structure
    IntValue:int=-1
    floatValue:float=0
    BoolValue:bool=false
   ) -> (then, ReturnValue:Fort Analytics Event Attribute Structure)
   > FireAnalyticsEvent_Blueprint!(execute, EventName:String=ItemReceivedPrePromptScreenSeen, Attribute:"Fort Analytics Event Attribute Structure (by ref)"=@n7) -> (then)
   Event OnInitialized()
   > "SetText (Text)"(self=Get Text_UserMsg())
   n8 = Event OnGiftBoxItemSet!() -> (OutputDelegate, then, bFromSelf:bool) {is=title K2Node_Event}
   > UpdateFromGiftBoxItem!(execute, self, bFromSelf:bool=@n8.bFromSelf) -> (then)
   Call On Pre Prompt Closed!(execute:Exec=@n3.then, self:Item Received Pre Prompt Object Reference) -> (then:Exec) {is=title K2Node_CallDelegate}
  ]}
 LargeImageLoaded {nodes [
   n1 = LargeImageLoaded!() -> (then, Obj:Object Reference) {is=title K2Node_FunctionEntry}
   n2 = Cast To Texture2D(Object=@n1)
   > SetBrushFromTexture!(execute, self:Image Object Reference, Texture:Texture 2D Object Reference=@n2, bMatchSize:bool=false) -> (then)
   n3 = Cast To Material!(execute=@n2.CastFailed, Object=@n1) -> (then, CastFailed, AsMaterial:Material Object Reference) {is=title K2Node_DynamicCast}
   > SetBrushFromMaterial(Material=@n3)
  ]}
 UpdateFromGiftBoxItem {nodes [
   n1 = UpdateFromGiftBoxItem!() -> (then, bFromSelf:bool=~) {is=title K2Node_FunctionEntry}
   > n2 = Set fromSelf(fromSelf:bool=@n1) -> (Output_Get:bool)
   n3 = Get GiftBoxItem()
   > SetVisibility(
    self=Get Overlay_GiftTag() -> (Overlay_GiftTag:Overlay Object Reference)
    InVisibility=Select(
     Option 0:ESlateVisibility Enum=Collapsed
     Option 1:ESlateVisibility Enum=Visible
     Index=HasParam!(self:Fort Gift Box Item Object Reference=@n3, ParamName:String=userMessage) -> (ReturnValue:bool)
    ) -> (ReturnValue:ESlateVisibility Enum)
   )
   > FilterUserMessageForProfanity!(
    execute
    self:Fort Gift Box Item Object Reference=@n3
    ControllerId:int=Get Local Player Controller ID!(Player:Player Controller Object Reference=GetOwningPlayer()) -> (ReturnValue:int=0) {is=title K2Node_CallFunction}
    OnTextFiltered:Delegate=Create Event()
   ) -> (then)
   > n4 = Branch(Condition=@n2)
   n5 = Get TopTitle() -> (TopTitle:Common Text Object Reference)
   > n6 = "SetText (Text)"(self=@n5, InText="LOCTEXT(\"Someone wonderful sent you a gift!\")")
   n7 = Get FromUsername()
   "SetText (Text)"(
    execute=[@n9.then, @n6.then]
    self=Get CommonTextBlock_0() -> (CommonTextBlock_0:Common Text Object Reference)
    InText=Select(Option 0:Text="To Text (String)"(InString=@n7), Option 1:Text="LOCTEXT(\"Yourself!\")", Index=@n2) -> (ReturnValue:Text)
   )
   n8 = Get GiftBoxItem()
   > SetBrushFromLazyDisplayAsset!(
    execute
    self:Common Lazy Image Object Reference=Get Image_GiftBox() -> (Image_GiftBox:Common Lazy Image Object Reference)
    LazyObject:"Soft Object Reference (by ref)"=Get ItemDisplayAsset!(self:Fort Gift Box Item Definition Object Reference=GetGiftBoxDefinition!(self:Fort Gift Box Item Object Reference=@n8) -> (ReturnValue:Fort Gift Box Item Definition Object Reference)) -> (ItemDisplayAsset:Soft Object Reference)
    bMatchTextureSize:bool=false
   ) -> (then)
   n9 = "SetText (Text)"(execute=@n4.else, self=@n5, InText="LOCTEXT(\"You received a gift!\")")
   n10 = Get GiftBoxItem()
   "SetText (Text)"(InText="To Text (String)"(InString=Get!(self:Fort Gift Box Item Object Reference=@n10) -> (GiftBoxInfo_GiftedOn:Date Time Structure, GiftBoxInfo_FromAccountId:String, GiftBoxInfo_LootList:Array of Mcp Loot Entry Structures).GiftBoxInfo_FromAccountId))
  ]}
]
variables [fromSelf {type=boolean}]
widgets {
 info {inheritedWidgetCount=2, parentClass=/Script/FortniteUI.FortItemReceivedPrePrompt, rootWidgetClass=&BackgroundBlur, widgetCount=34}
 widgets [
  BackgroundBlur_0 {
   widgetClassPath=&BackgroundBlur
   widget=BackgroundBlur_0
   properties {
    blurStrength=25
    lowQualityFallbackBrush {
     bIsDynamicallyLoaded=false
     drawAs=NoDrawType
     imageType=NoImage
     mirroring=NoMirror
     resourceName=None
     resourceObject=null
     tiling=NoTile
     imageSize=<0 0>
     margin=<ltrb 0 0 0 0>
     outlineSettings {
      bUseBrushTransparency=false
      roundingType=HalfHeightRadius
      width=0
      color {colorUseRule=UseColor_Specified, specifiedColor=<rgba 0 0 0 0>}
      cornerRadii=<0 0 0 0>
     }
     tintColor {colorUseRule=UseColor_Specified, specifiedColor=<rgba 0 0 1 1>}
     uVRegion {bIsValid=false, max=<0 0>, min=<0 0>}
    }
   }
  }
  MainSafeZone {
   widgetClassPath=&SafeZone
   parent=BackgroundBlur_0
   widget=MainSafeZone
   properties {padBottom=false, padTop=false}
   slot {class=BackgroundBlurSlot, properties {padding=<ltrb 0 0 0 0>}}
  }
  HorizontalBox_2 {
   widgetClassPath=&HorizontalBox
   parent=MainSafeZone
   widget=HorizontalBox_2
   properties {visibility=Visible}
   slot {class=SafeZoneSlot}
  }
  VerticalBox_8 {
   widgetClassPath=&VerticalBox
   parent=HorizontalBox_2
   widget=VerticalBox_8
   slot {class=HorizontalBoxSlot, properties {horizontalAlignment=HAlign_Center, padding=<ltrb 0 120 0 100>, size {sizeRule=Fill, value=1}}}
  }
  HorizontalBox_3 {
   widgetClassPath=&HorizontalBox
   parent=VerticalBox_8
   widget=HorizontalBox_3
   slot {class=VerticalBoxSlot, properties {horizontalAlignment=HAlign_Center}}
  }
  Spacer_42 {widgetClassPath=&Spacer, parent=HorizontalBox_3, widget=Spacer_42, properties {size=<24 1>}, slot {class=HorizontalBoxSlot}}
  BorderSuccessText {
   widgetClassPath=&Border
   parent=HorizontalBox_3
   widget=BorderSuccessText
   properties {
    background {
     bIsDynamicallyLoaded=false
     drawAs=NoDrawType
     imageType=NoImage
     mirroring=NoMirror
     resourceName=None
     resourceObject=null
     tiling=NoTile
     imageSize=<32 32>
     margin=<ltrb 0 0 0 0>
     outlineSettings {
      bUseBrushTransparency=false
      roundingType=HalfHeightRadius
      width=0
      color {colorUseRule=UseColor_Specified, specifiedColor=<rgba 0 0 0 0>}
      cornerRadii=<0 0 0 0>
     }
     tintColor {colorUseRule=UseColor_Specified, specifiedColor=<rgba 1 1 1 1>}
     uVRegion {bIsValid=false, max=<0 0>, min=<0 0>}
    }
    padding=<ltrb 0 0 0 0>
   }
   slot {class=HorizontalBoxSlot, properties {verticalAlignment=VAlign_Center}}
  }
  BorderCheck {
   widgetClassPath=&Border
   parent=HorizontalBox_3
   widget=BorderCheck
   properties {
    background {
     bIsDynamicallyLoaded=false
     drawAs=NoDrawType
     imageType=NoImage
     mirroring=NoMirror
     resourceName=None
     resourceObject=null
     tiling=NoTile
     imageSize=<32 32>
     margin=<ltrb 0 0 0 0>
     outlineSettings {
      bUseBrushTransparency=false
      roundingType=HalfHeightRadius
      width=0
      color {colorUseRule=UseColor_Specified, specifiedColor=<rgba 0 0 0 0>}
      cornerRadii=<0 0 0 0>
     }
     tintColor {colorUseRule=UseColor_Specified, specifiedColor=<rgba 1 1 1 1>}
     uVRegion {bIsValid=false, max=<0 0>, min=<0 0>}
    }
    padding=<ltrb 0 0 0 0>
   }
   slot {class=HorizontalBoxSlot, properties {horizontalAlignment=HAlign_Center, verticalAlignment=VAlign_Center, padding=<ltrb 4 0 0 0>}}
  }
  HeaderContainer {
   widgetClassPath=&HorizontalBox
   bInherited=true
   bIsVariable=true
   parent=VerticalBox_8
   widget=HeaderContainer
   slot {class=VerticalBoxSlot}
  }
  GiftScrollBox {widgetClassPath=&ScrollBox, bInherited=true, parent=HeaderContainer, widget=GiftScrollBox, slot {class=HorizontalBoxSlot}}
  BorderItemNAme {
   widgetClassPath=&Border
   parent=VerticalBox_8
   widget=BorderItemNAme
   properties {
    background {
     bIsDynamicallyLoaded=false
     drawAs=NoDrawType
     imageType=NoImage
     mirroring=NoMirror
     resourceName=None
     resourceObject=null
     tiling=NoTile
     imageSize=<32 32>
     margin=<ltrb 0 0 0 0>
     outlineSettings {
      bUseBrushTransparency=false
      roundingType=HalfHeightRadius
      width=0
      color {colorUseRule=UseColor_Specified, specifiedColor=<rgba 0 0 0 0>}
      cornerRadii=<0 0 0 0>
     }
     tintColor {colorUseRule=UseColor_Specified, specifiedColor=<rgba 1 1 1 1>}
     uVRegion {bIsValid=false, max=<0 0>, min=<0 0>}
    }
    padding=<ltrb 0 0 0 0>
   }
   slot {class=VerticalBoxSlot, properties {horizontalAlignment=HAlign_Center, verticalAlignment=VAlign_Center}}
  }
  Overlay_9 {widgetClassPath=&Overlay, parent=BorderItemNAme, widget=Overlay_9, slot {class=BorderSlot, properties {padding=<ltrb 0 0 0 0>}}}
  VerticalBox_0 {
   widgetClassPath=&VerticalBox
   parent=Overlay_9
   widget=VerticalBox_0
   slot {class=OverlaySlot, properties {horizontalAlignment=HAlign_Center, verticalAlignment=VAlign_Center}}
  }
  TopTitle {
   widgetClassPath=&CommonTextBlock
   bIsVariable=true
   parent=VerticalBox_0
   widget=TopTitle
   properties {
    justification=Center
    scrollStyle=&TextScrollStyle-Base_C
    style=/FortUILibrary/Styles/Typography/Foundation/TextStyle-Header-L-Horizon-Yellow.TextStyle-Header-L-Horizon-Yellow_C
    text="You received a gift!"
    font {
     bForceMonospaced=false
     bMaterialIsStencil=false
     fontMaterial=/FortUILibrary/Materials/Foundation/M_Font_Horizon_Yellow
     fontObject=&Burbank
     letterSpacing=0
     monospacedWidth=1
     size=48
     skewAmount=0
     typefaceFontName=Burbank-Black
     outlineSettings {
      bApplyOutlineToDropShadows=true
      bMiteredCorners=false
      bSeparateFillAlpha=false
      outlineMaterial=null
      outlineSize=3
      outlineColor=<rgba 0 0 0 0.1>
     }
    }
    shadowOffset=<0 0>
   }
   slot {class=VerticalBoxSlot, properties {horizontalAlignment=HAlign_Center, padding=<ltrb 0 -20 0 0>}}
  }
  FromField {
   widgetClassPath=&HorizontalBox
   bIsVariable=true
   parent=VerticalBox_0
   widget=FromField
   slot {class=VerticalBoxSlot, properties {horizontalAlignment=HAlign_Center}}
  }
  Textfrom {
   widgetClassPath=&CommonTextBlock
   bIsVariable=true
   parent=FromField
   widget=Textfrom
   properties {
    justification=Center
    scrollStyle=&TextScrollStyle-Base_C
    style=/FortUILibrary/Styles/Typography/Foundation/TextStyle-Base-L.TextStyle-Base-L_C
    text="from: "
    font {
     bForceMonospaced=false
     bMaterialIsStencil=false
     fontMaterial=null
     fontObject=&NotoSans
     letterSpacing=0
     monospacedWidth=1
     size=32
     skewAmount=0
     typefaceFontName=NotoSans-Regular
     outlineSettings {
      bApplyOutlineToDropShadows=false
      bMiteredCorners=false
      bSeparateFillAlpha=false
      outlineMaterial=null
      outlineSize=0
      outlineColor=<rgba 0 0 0 1>
     }
    }
    shadowOffset=<0 0>
   }
   slot {class=HorizontalBoxSlot}
  }
  CommonTextBlock_0 {
   widgetClassPath=&CommonTextBlock
   bIsVariable=true
   parent=FromField
   widget=CommonTextBlock_0
   properties {
    justification=Center
    scrollStyle=&TextScrollStyle-Base_C
    style=/FortUILibrary/Styles/Typography/Foundation/TextStyle-Base-L-B.TextStyle-Base-L-B_C
    text=Sree_Devy
    font {
     bForceMonospaced=false
     bMaterialIsStencil=false
     fontMaterial=null
     fontObject=&NotoSans
     letterSpacing=0
     monospacedWidth=1
     size=32
     skewAmount=0
     typefaceFontName=NotoSans-Bold
     outlineSettings {
      bApplyOutlineToDropShadows=false
      bMiteredCorners=false
      bSeparateFillAlpha=false
      outlineMaterial=null
      outlineSize=0
      outlineColor=<rgba 0 0 0 1>
     }
    }
    shadowOffset=<0 0>
   }
   slot {class=HorizontalBoxSlot, properties {horizontalAlignment=HAlign_Center}}
  }
  Overlay_0 {
   widgetClassPath=&Overlay
   parent=VerticalBox_8
   widget=Overlay_0
   slot {class=VerticalBoxSlot, properties {horizontalAlignment=HAlign_Center, verticalAlignment=VAlign_Center, size {sizeRule=Fill, value=1}}}
  }
  BorderRarityFlare {
   widgetClassPath=&Border
   parent=Overlay_0
   widget=BorderRarityFlare
   properties {
    horizontalAlignment=HAlign_Center
    verticalAlignment=VAlign_Center
    background {
     bIsDynamicallyLoaded=false
     drawAs=NoDrawType
     imageType=NoImage
     mirroring=NoMirror
     resourceName=None
     resourceObject=null
     tiling=NoTile
     imageSize=<32 32>
     margin=<ltrb 0 0 0 0>
     outlineSettings {
      bUseBrushTransparency=false
      roundingType=HalfHeightRadius
      width=0
      color {colorUseRule=UseColor_Specified, specifiedColor=<rgba 0 0 0 0>}
      cornerRadii=<0 0 0 0>
     }
     tintColor {colorUseRule=UseColor_Specified, specifiedColor=<rgba 1 1 1 1>}
     uVRegion {bIsValid=false, max=<0 0>, min=<0 0>}
    }
    padding=<ltrb -100 -100 -100 -100>
   }
   slot {class=OverlaySlot, properties {horizontalAlignment=HAlign_Fill, verticalAlignment=VAlign_Fill}}
  }
  ImageRarityFlare {
   widgetClassPath=&Image
   bIsVariable=true
   parent=BorderRarityFlare
   widget=ImageRarityFlare
   properties {
    visibility=SelfHitTestInvisible
    brush {
     bIsDynamicallyLoaded=false
     drawAs=Image
     imageType=NoImage
     mirroring=NoMirror
     resourceName=None
     resourceObject=/Game/UI/Frontend/Rewards/materials/M_UI_DailyReward_Rarity-Flares_Gift
     tiling=NoTile
     imageSize=<260.1383 363.9082>
     margin=<ltrb 0 0 0 0>
     outlineSettings {
      bUseBrushTransparency=false
      roundingType=HalfHeightRadius
      width=0
      color {colorUseRule=UseColor_Specified, specifiedColor=<rgba 0 0 0 0>}
      cornerRadii=<0 0 0 0>
     }
     tintColor {colorUseRule=UseColor_Specified, specifiedColor=<rgba 1 1 1 1>}
     uVRegion {bIsValid=false, max=<0 0>, min=<0 0>}
    }
    renderTransform {angle=0, scale=<2.75 1.875>, shear=<0 0>, translation=<0 24>}
   }
   slot {class=BorderSlot, properties {horizontalAlignment=HAlign_Center, verticalAlignment=VAlign_Center, padding=<ltrb -100 -100 -100 -100>}}
  }
  OverlayTriangleParticles {
   widgetClassPath=&Overlay
   parent=Overlay_0
   widget=OverlayTriangleParticles
   properties {visibility=HitTestInvisible, renderTransform {angle=0, scale=<0.8 0.9>, shear=<0 0>, translation=<0 0>}}
   slot {class=OverlaySlot, properties {horizontalAlignment=HAlign_Center, verticalAlignment=VAlign_Center, padding=<ltrb -200 -200 -200 -200>}}
  }
  ImageTriangleL {
   widgetClassPath=&Image
   parent=OverlayTriangleParticles
   widget=ImageTriangleL
   properties {
    brush {
     bIsDynamicallyLoaded=false
     drawAs=Image
     imageType=NoImage
     mirroring=NoMirror
     resourceName=None
     resourceObject=/Game/UI/HUD/Materials/M_FX_TriangleSpiral
     tiling=NoTile
     imageSize=<320.4127 251.0823>
     margin=<ltrb 0 0 0 0>
     outlineSettings {
      bUseBrushTransparency=false
      roundingType=HalfHeightRadius
      width=0
      color {colorUseRule=UseColor_Specified, specifiedColor=<rgba 0 0 0 0>}
      cornerRadii=<0 0 0 0>
     }
     tintColor {colorUseRule=UseColor_Specified, specifiedColor=<rgba 1 0.7937 0.3718 1>}
     uVRegion {bIsValid=false, max=<0 0>, min=<0 0>}
    }
    renderTransform {angle=0, scale=<-1 1>, shear=<0 0>, translation=<0 0>}
    renderTransformPivot=<0 1>
   }
   slot {class=OverlaySlot, properties {horizontalAlignment=HAlign_Right}}
  }
  ImageTriangleR {
   widgetClassPath=&Image
   parent=OverlayTriangleParticles
   widget=ImageTriangleR
   properties {
    brush {
     bIsDynamicallyLoaded=false
     drawAs=Image
     imageType=NoImage
     mirroring=NoMirror
     resourceName=None
     resourceObject=&M_FX_TriangleSpiral
     tiling=NoTile
     imageSize=<362.1116 243.2954>
     margin=<ltrb 0 0 0 0>
     outlineSettings {
      bUseBrushTransparency=false
      roundingType=HalfHeightRadius
      width=0
      color {colorUseRule=UseColor_Specified, specifiedColor=<rgba 0 0 0 0>}
      cornerRadii=<0 0 0 0>
     }
     tintColor {colorUseRule=UseColor_Specified, specifiedColor=<rgba 1 1 1 1>}
     uVRegion {bIsValid=false, max=<0 0>, min=<0 0>}
    }
    renderTransform {angle=0, scale=<1 -1>, shear=<0 0>, translation=<292 0>}
    renderTransformPivot=<1 0.5>
   }
   slot {class=OverlaySlot}
  }
  ImageTriangleU {
   widgetClassPath=&Image
   parent=OverlayTriangleParticles
   widget=ImageTriangleU
   properties {
    brush {
     bIsDynamicallyLoaded=false
     drawAs=Image
     imageType=NoImage
     mirroring=NoMirror
     resourceName=None
     resourceObject=&M_FX_TriangleSpiral
     tiling=NoTile
     imageSize=<374.9563 236.9302>
     margin=<ltrb 0 0 0 0>
     outlineSettings {
      bUseBrushTransparency=false
      roundingType=HalfHeightRadius
      width=0
      color {colorUseRule=UseColor_Specified, specifiedColor=<rgba 0 0 0 0>}
      cornerRadii=<0 0 0 0>
     }
     tintColor {colorUseRule=UseColor_Specified, specifiedColor=<rgba 1 1 1 1>}
     uVRegion {bIsValid=false, max=<0 0>, min=<0 0>}
    }
    renderTransform {angle=-90, scale=<1 1>, shear=<0 0>, translation=<0 -220>}
   }
   slot {class=OverlaySlot, properties {horizontalAlignment=HAlign_Center, verticalAlignment=VAlign_Bottom}}
  }
  ImageTriangleD {
   widgetClassPath=&Image
   parent=OverlayTriangleParticles
   widget=ImageTriangleD
   properties {
    brush {
     bIsDynamicallyLoaded=false
     drawAs=Image
     imageType=NoImage
     mirroring=NoMirror
     resourceName=None
     resourceObject=&M_FX_TriangleSpiral
     tiling=NoTile
     imageSize=<380.916 251.0823>
     margin=<ltrb 0 0 0 0>
     outlineSettings {
      bUseBrushTransparency=false
      roundingType=HalfHeightRadius
      width=0
      color {colorUseRule=UseColor_Specified, specifiedColor=<rgba 0 0 0 0>}
      cornerRadii=<0 0 0 0>
     }
     tintColor {colorUseRule=UseColor_Specified, specifiedColor=<rgba 1 1 1 1>}
     uVRegion {bIsValid=false, max=<0 0>, min=<0 0>}
    }
    renderTransform {angle=90, scale=<1 1>, shear=<0 0>, translation=<0 282>}
   }
   slot {class=OverlaySlot, properties {horizontalAlignment=HAlign_Center}}
  }
  SizeBoxGiftBox {
   widgetClassPath=&SizeBox
   parent=Overlay_0
   widget=SizeBoxGiftBox
   properties {renderTransformPivot=<0.5 0.9>}
   slot {class=OverlaySlot, properties {horizontalAlignment=HAlign_Center}}
  }
  Image_GiftBox {
   widgetClassPath=/Script/CommonUI.CommonLazyImage
   bIsVariable=true
   parent=SizeBoxGiftBox
   widget=Image_GiftBox
   properties {brush {
     bIsDynamicallyLoaded=false
     drawAs=Image
     imageType=NoImage
     mirroring=NoMirror
     resourceName=None
     resourceObject=/Game/Athena/UI/Frontend/Gifting/Materials/M_StoreGift_var0
     tiling=NoTile
     imageSize=<512 512>
     margin=<ltrb 0 0 0 0>
     outlineSettings {
      bUseBrushTransparency=false
      roundingType=HalfHeightRadius
      width=0
      color {colorUseRule=UseColor_Specified, specifiedColor=<rgba 0 0 0 0>}
      cornerRadii=<0 0 0 0>
     }
     tintColor {colorUseRule=UseColor_Specified, specifiedColor=<rgba 1 1 1 1>}
     uVRegion {bIsValid=false, max=<0 0>, min=<0 0>}
    }}
   slot {class=SizeBoxSlot}
  }
  Overlay_GiftTag {
   widgetClassPath=&Overlay
   bIsVariable=true
   parent=Overlay_0
   widget=Overlay_GiftTag
   properties {renderTransform {angle=-7, scale=<1 1>, shear=<0 0>, translation=<504 0>}, renderTransformPivot=<0 0.8>}
   slot {class=OverlaySlot, properties {horizontalAlignment=HAlign_Right, verticalAlignment=VAlign_Center}}
  }
  Image_GiftTagShadow2 {
   widgetClassPath=&Image
   parent=Overlay_GiftTag
   widget=Image_GiftTagShadow2
   properties {
    brush {
     bIsDynamicallyLoaded=false
     drawAs=Image
     imageType=NoImage
     mirroring=NoMirror
     resourceName=None
     resourceObject=/Game/UI/ItemReceived/Materials/M_UI_GiftTag_Filled
     tiling=NoTile
     imageSize=<650 325>
     margin=<ltrb 0 0 0 0>
     outlineSettings {
      bUseBrushTransparency=false
      roundingType=HalfHeightRadius
      width=0
      color {colorUseRule=UseColor_Specified, specifiedColor=<rgba 0 0 0 0>}
      cornerRadii=<0 0 0 0>
     }
     tintColor {colorUseRule=UseColor_Specified, specifiedColor=<rgba 1 1 1 1>}
     uVRegion {bIsValid=false, max=<0 0>, min=<0 0>}
    }
    colorAndOpacity=<rgba 0.0052 0.0052 0.0056 0.227>
    renderTransform {angle=0, scale=<1 1>, shear=<0 0>, translation=<-5 28>}
   }
   slot {class=OverlaySlot}
  }
  Image_GiftTagShadow {
   widgetClassPath=&Image
   parent=Overlay_GiftTag
   widget=Image_GiftTagShadow
   properties {
    brush {
     bIsDynamicallyLoaded=false
     drawAs=Image
     imageType=NoImage
     mirroring=NoMirror
     resourceName=None
     resourceObject=&M_UI_GiftTag_Filled
     tiling=NoTile
     imageSize=<650 325>
     margin=<ltrb 0 0 0 0>
     outlineSettings {
      bUseBrushTransparency=false
      roundingType=HalfHeightRadius
      width=0
      color {colorUseRule=UseColor_Specified, specifiedColor=<rgba 0 0 0 0>}
      cornerRadii=<0 0 0 0>
     }
     tintColor {colorUseRule=UseColor_Specified, specifiedColor=<rgba 1 1 1 1>}
     uVRegion {bIsValid=false, max=<0 0>, min=<0 0>}
    }
    colorAndOpacity=<rgba 0.1046 0.1413 0.1878 1>
    renderTransform {angle=0, scale=<1 1>, shear=<0 0>, translation=<0 9>}
   }
   slot {class=OverlaySlot}
  }
  Image_GiftTag {
   widgetClassPath=&Image
   parent=Overlay_GiftTag
   widget=Image_GiftTag
   properties {brush {
     bIsDynamicallyLoaded=false
     drawAs=Image
     imageType=NoImage
     mirroring=NoMirror
     resourceName=None
     resourceObject=/Game/UI/ItemReceived/Materials/M_UI_GiftTag_Inst
     tiling=NoTile
     imageSize=<650 325>
     margin=<ltrb 0 0 0 0>
     outlineSettings {
      bUseBrushTransparency=false
      roundingType=HalfHeightRadius
      width=0
      color {colorUseRule=UseColor_Specified, specifiedColor=<rgba 0 0 0 0>}
      cornerRadii=<0 0 0 0>
     }
     tintColor {colorUseRule=UseColor_Specified, specifiedColor=<rgba 1 1 1 1>}
     uVRegion {bIsValid=false, max=<0 0>, min=<0 0>}
    }}
   slot {class=OverlaySlot}
  }
  UserMessageBorder {
   widgetClassPath=&CommonBorder
   bIsVariable=true
   parent=Overlay_GiftTag
   widget=UserMessageBorder
   properties {
    background {
     bIsDynamicallyLoaded=false
     drawAs=Box
     imageType=NoImage
     mirroring=NoMirror
     resourceName=None
     resourceObject=/FortUILibrary/Styles/Border/Materials/M-Border-MainL-Black
     tiling=NoTile
     imageSize=<128 32>
     margin=<ltrb 0.1 0.45 0.9 0.5>
     outlineSettings {
      bUseBrushTransparency=false
      roundingType=HalfHeightRadius
      width=0
      color {colorUseRule=UseColor_Specified, specifiedColor=<rgba 0 0 0 0>}
      cornerRadii=<0 0 0 0>
     }
     tintColor {colorUseRule=UseColor_Specified, specifiedColor=<rgba 1 1 1 1>}
     uVRegion {bIsValid=false, max=<0 0>, min=<0 0>}
    }
    brushColor=<rgba 1 1 1 0>
    padding=<ltrb 20 20 20 20>
   }
   slot {class=OverlaySlot, properties {horizontalAlignment=HAlign_Fill, verticalAlignment=VAlign_Fill, padding=<ltrb 80 35 0 35>}}
  }
  Text_UserMsg {
   widgetClassPath=&CommonTextBlock
   bIsVariable=true
   parent=UserMessageBorder
   widget=Text_UserMsg
   properties {
    text="{User Message}"
    wrapTextAt=472.9804
    wrappingPolicy=AllowPerCharacterWrapping
    colorAndOpacity {colorUseRule=UseColor_Specified, specifiedColor=<rgba 0.0052 0.0052 0.0056 1>}
    font {
     bForceMonospaced=false
     bMaterialIsStencil=false
     fontMaterial=null
     fontObject=&BurbankSmall
     letterSpacing=0
     monospacedWidth=1
     size=24
     skewAmount=0
     typefaceFontName=NotoSans-Regular
     outlineSettings {
      bApplyOutlineToDropShadows=false
      bMiteredCorners=false
      bSeparateFillAlpha=false
      outlineMaterial=null
      outlineSize=0
      outlineColor=<rgba 0 0 0 1>
     }
    }
    renderTransform {angle=0, scale=<1 1>, shear=<2 0>, translation=<0 0>}
   }
   slot {class=BorderSlot, properties {padding=<ltrb 20 20 20 20>}}
  }
  OpenButton {
   widgetClassPath=&IconTextButton
   bIsVariable=true
   parent=VerticalBox_8
   widget=OpenButton
   properties {
    bDisplayInputActionWhenNotInteractable=false
    buttonText="OPEN IT!"
    contentAlignment=HAlign_Center
    minHeight=75
    minWidth=343
    onButtonBaseClicked=(&self:WidgetTree.OpenButton.HandleOnButtonClicked)
    onButtonBaseDoubleClicked=(&self:WidgetTree.OpenButton.HandleOnButtonDoubleClicked)
    onButtonBaseHovered=(&self:WidgetTree.OpenButton.HandleOnButtonHovered)
    onButtonBaseUnhovered=(&self:WidgetTree.OpenButton.HandleOnButtonUnhovered)
    onSelectedChangedBase=(&self:WidgetTree.OpenButton.HandleOnSelectedChanged)
    style=/FortUILibrary/Foundation/Buttons/ButtonStyles/Skew/ButtonStyle-Skew_Desirable.ButtonStyle-Skew_Desirable_C
    visibility=Hidden
    hoveredSlateSoundOverride {resourceObject=/Game/Sounds/Fort_UI_Sounds/FrontEnd/New_FrontEnd/Fortnite_UI_VBucks_Menu_Hover_03_Cue}
    triggeringInputAction {dataTable=&InputActionDataTable, rowName=RewardsOpen}
   }
   slot {class=VerticalBoxSlot, properties {horizontalAlignment=HAlign_Center, padding=<ltrb 0 50 0 0>}}
  }
 ]
}
