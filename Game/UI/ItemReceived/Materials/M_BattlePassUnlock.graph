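# Translucent UI material for a battle pass unlock effect in the item-received screen, drawing the battle pass icon with animated outlines, fill and shine.
# Red, green and blue time parameters threshold an outline channel texture to reveal lines, while cloud noise dissolves in the icon fill.
# An offset-driven shine, rotated and bump-offset by the squared, desaturated icon, color-dodges over it.
# Whiteout, radial line overlay and opacity-switch parameters blend the layers.
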Material /Game/UI/ItemReceived/Materials/M_BattlePassUnlock
graph {
 properties {
  bUsedWithStaticMesh=false
  blendMode=BLEND_Translucent
  materialDomain=MD_UI
  pixelDepthOffsetMode=PDOM_Legacy
  usedShadingModels=Default Lit
 }
 nodes [
  n1 = TextureSample() {texture=/Game/UI/Foundation/Textures/Icons/Items/T-FNBR-BattlePass-L}
  n2 = MaterialFunctionCall!(
   HeightTexture=Desaturation(None=Multiply(A=@n1.RGB, B=@n1.RGB))
   TextureCoordinates
   ReferencePlane
   HeightRatioInput
   CameraVector=VectorParameter() {parameterName=BumpCameraVector}.RGB
  ) {materialFunction=/Engine/Functions/Engine_MaterialFunctions02/Utility/BumpOffset_Advanced.BumpOffset_advanced}
  n3 = Constant2Vector()
  n4 = Constant() {r=0.9}
  n5 = ScalarParameter() {parameterName=ShineOffset, sliderMax=1, sliderMin=-1}
  n6 = Constant()
  n7 = TextureSample(UVs=Add(
    A=MaterialFunctionCall(UVs=@n2.UVs, Rotation Center=@n3, "Rotation Angle (0-1)"=@n4) {materialFunction=&CustomRotator}.Rotated Values
    B=AppendVector(A=@n5, B=@n6)
   )) {samplerSource=SSM_Clamp_WorldGroupSettings, texture=/Game/UI/Foundation/Textures/Pattern/T-UI-Shine}
  Add(A=@n1.RGB, B=@n7.RGB)
  n8 = TextureSample() {
   samplerSource=SSM_Clamp_WorldGroupSettings
   samplerType=SAMPLERTYPE_LinearColor
   texture=/Game/UI/ItemReceived/Materials/Textures/T_Athena_BattlePassAndStar_OutLineChannels
  }
  n9 = Max(A=@n8.R, B=@n8.R)
  n10 = Add(A=Add(A=@n8.R, B=@n8.R), B=@n8.B)
  n11 = Constant()
  n12 = If(
   A=OneMinus(None=@n8.B)
   B=ScalarParameter() {defaultValue=1, parameterName=BlueTime, sliderMax=1}
   "A > B"=@n11
   "A == B"=@n11
   "A < B"=Constant() {r=1}
  )
  n13 = Constant()
  n14 = If(
   A=OneMinus(None=@n8.G)
   B=ScalarParameter() {defaultValue=1, parameterName=GreenTime, sliderMax=1}
   "A > B"=@n13
   "A == B"=@n13
   "A < B"=Constant() {r=1}
  )
  n15 = OneMinus(None=@n8.R)
  n16 = Constant()
  n17 = Constant() {r=1}
  n18 = ScalarParameter() {defaultValue=1, parameterName=RedTime, sliderMax=1}
  n19 = LinearInterpolate(
   B=Max(A=Max(A=@n9, B=@n8.B), B=Clamp(None=@n10))
   Alpha=Max(A=Max(A=@n12, B=@n14), B=If(A=@n15, B=@n18, "A > B"=@n16, "A == B"=@n16, "A < B"=@n17))
  )
  n20 = Multiply(A=@n19, B=ScalarParameter() {defaultValue=1, parameterName=LineMax})
  n21 = TextureSample() {texture=&AtmosphericCloudNoise01}
  n22 = Constant() {r=1}
  n23 = Multiply(
   A=If(A=OneMinus(None=@n21.G), B=ScalarParameter() {parameterName=ImageFillAmount, sliderMax=1}, "A > B"=Constant(), "A == B"=@n22, "A < B"=@n22)
   B=@n1.A
  )
  LinearInterpolate(
   A=LinearInterpolate(A=@n19, B=Max(A=@n20, B=@n23), Alpha=ScalarParameter() {defaultValue=1, parameterName=ImageFill})
   B=@n1.A
   Alpha=ScalarParameter() {defaultValue=1, parameterName=OpacitySwitch, sliderMax=1}
  )
  n24 = LinearInterpolate(
   A=Clamp(None=MaterialFunctionCall!(Blend=@n7.RGB, Base=@n1.RGB) {materialFunction=/Engine/Functions/Engine_MaterialFunctions03/Blends/Blend_ColorDodge}.Result)
   B=Constant() {r=1}
   Alpha=ScalarParameter() {parameterName=Whiteout, sliderMax=1}
  )
  n25 = ScalarParameter() {defaultValue=1.4857, parameterName=LineOverlay, sliderMax=1.5}
  n26 = Constant() {r=3}
  LinearInterpolate(
   A=@n24
   B=Add(A=@n24, B=@n19)
   Alpha=OneMinus(None=MaterialFunctionCall(Radius=@n25, Density=@n26) {materialFunction=&RadialGradientExponential}.RadialGradientExponential)
  )
  Time() {bOverride_Period=true, period=1.1}
 ]
}
