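# Translucent UI material for the battle pass purchase screen transition, drawing a two-palette gradient background with animated triangles.
# An aspect-corrected radial gradient shapes both color pairs and masks a rotated, screen-aligned triangle tile whose channels crossfade on a clamped sine.
# The triangles color-dodge onto each palette, and TransitionBackground blends between the results.
# Panning noise drives lerps with identical inputs, so it has no visible effect.
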Material /Game/UI/ItemReceived/Materials/M_UI_BattlePassPurchase_Screen_Transition
graph {
 properties {
  bUsedWithStaticMesh=false
  blendMode=BLEND_Translucent
  materialDomain=MD_UI
  pixelDepthOffsetMode=PDOM_Legacy
  usedShadingModels=Default Lit
 }
 nodes [
  Constant() {r=1.5}
  n1 = MaterialFunctionCall(Float2=MaterialFunctionCall() {materialFunction=&ScreenResolution}.Visible Resolution) {materialFunction=&BreakOutFloat2Components}
  n2 = Max(A=@n1.G) {constB=0.1}
  n3 = MaterialFunctionCall() {materialFunction=&ScreenAlignedUVs}
  n4 = Multiply(A=AppendVector(A=Divide(A=@n1.R, B=@n2), B=Constant() {r=1}), B=@n3."X 100%, Y 100%") {constB=10}
  n5 = AppendVector(
   A=ScalarParameter() {defaultValue=4, group=TriangleBackground, parameterName=TriangleTiles-U}
   B=ScalarParameter() {defaultValue=2.5, group=TriangleBackground, parameterName=TriangleTiles-V}
  )
  n6 = TextureSample(UVs=Rotator(
    Coordinate=Multiply(A=@n4, B=@n5) {constB=20}
    Time=ScalarParameter() {defaultValue=0.125, group=TriangleBackground, parameterName=TriangleRotation}
   ) {speed=6.2832}) {texture=/Game/UI/Frontend/Campaign/EventCallOut/T-UI-TriangleTile}
  n7 = AppendVector(
   A=ScalarParameter() {defaultValue=1, group=TriangleBackground, parameterName=TriangleBlendTiles-U}
   B=ScalarParameter() {defaultValue=0.5, group=TriangleBackground, parameterName=TriangleBlendTiles-V}
  )
  n8 = Time()
  n9 = ScalarParameter() {defaultValue=1, group=TriangleBackground, parameterName=TimeMultiplier}
  n10 = TextureSample(UVs=Panner(Coordinate=Multiply(A=@n3."X 100%, Y 100%", B=@n7), Time=Multiply(A=@n8, B=@n9)) {speedX=-0.25, speedY=0.125}) {samplerType=SAMPLERTYPE_Grayscale, texture=&LowResBlurredNoise}
  n11 = LinearInterpolate(A=@n6.B, B=@n6.B, Alpha=@n10.R)
  n12 = LinearInterpolate(A=@n6.R, B=@n6.R, Alpha=@n10.R)
  n13 = Clamp(None=Sine(None=@n8) {period=4})
  n14 = Constant() {r=1}
  n15 = MaterialFunctionCall() {materialFunction=&GetUserInterfaceUV}
  n16 = MaterialFunctionCall(Float2=@n15.Pixel Size) {materialFunction=&BreakOutFloat2Components}
  n17 = Max(A=@n16.R)
  n18 = Divide(A=@n16.G, B=@n17)
  n19 = Constant2Vector()
  n20 = MaterialFunctionCall(
   UVs=Multiply(A=AppendVector(A=@n14, B=@n18), B=@n15.Normalized UV)
   CenterPosition=Constant2Vector() {g=0.25, r=0.5}
   Radius=Divide(A=Distance(A=@n19, B=@n15.Pixel Size), B=@n17)
   Density=ScalarParameter() {defaultValue=2.333, parameterName=Density}
  ) {materialFunction=&RadialGradientExponential}
  n21 = Saturate(None=OneMinus(None=@n20.RadialGradientExponential))
  n22 = ScalarParameter() {parameterName=TrianglesMask_Radius}
  n23 = Multiply(
   A=DotProduct(A=LinearInterpolate(A=@n11, B=@n12, Alpha=@n13), B=Power(Base=@n21, Exp=@n22) {constExponent=4})
   B=ScalarParameter() {defaultValue=0.5, parameterName=Triangles_Amount}
  )
  n24 = LinearInterpolate(
   A=VectorParameter() {parameterName=Color2, defaultValue=<rgba 0 0.0194 0.3185 1>}.RGB
   B=Multiply(
    A=VectorParameter() {parameterName=Color1, defaultValue=<rgba 0 0.0578 0.6038 1>}.RGB
    B=ScalarParameter() {defaultValue=1, parameterName=Hotspot_Intensity}
   )
   Alpha=@n20.RadialGradientExponential
  )
  n25 = LinearInterpolate(
   A=VectorParameter() {parameterName=Color4, defaultValue=<rgba 0 0.0194 0.3185 1>}.RGB
   B=Multiply(
    A=VectorParameter() {parameterName=Color3, defaultValue=<rgba 0 0.0578 0.6038 1>}.RGB
    B=ScalarParameter() {defaultValue=1, parameterName=Hotspot_Intensity}
   )
   Alpha=@n20.RadialGradientExponential
  )
  Saturate(None=LinearInterpolate(
    A=MaterialFunctionCall!(Blend=@n23, Base=@n24) {materialFunction=/Engine/Functions/Engine_MaterialFunctions03/Blends/Blend_ColorDodge}.Result
    B=MaterialFunctionCall(Blend=@n23, Base=@n25) {materialFunction=&Blend_ColorDodge}.Result
    Alpha=ScalarParameter() {parameterName=TransitionBackground}
   ))
  ScalarParameter() {defaultValue=1, parameterName=OverallAlpha}
  TextureCoordinate()
 ]
}
