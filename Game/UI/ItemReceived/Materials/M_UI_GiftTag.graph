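# UI material for the gift tag in the item-received UI, drawing a translucent shape with outer stroke, outline and inner fill.
# A texture's red channel feeds three anti-aliased thresholds at cumulative offsets, giving the shape mask and two band boundaries, with shared edge softening.
# Outer stroke and inner fill take gradients from a curve atlas along rotatable UVs, while the outline is a flat color.
# A WhiteFlood parameter blends the color to white.
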Material /Game/UI/ItemReceived/Materials/M_UI_GiftTag
graph {
 properties {
  bUsedWithStaticMesh=false
  blendMode=BLEND_Translucent
  materialDomain=MD_UI
  pixelDepthOffsetMode=PDOM_Legacy
  usedShadingModels=Default Lit
 }
 nodes [
  n1 = TextureCoordinate()
  n2 = ScalarParameter() {parameterName=OuterStrokeDirection, sliderMax=1}
  n3 = CurveAtlasRowParameter(CurveTime=ComponentMask(None=MaterialFunctionCall(UVs=@n1, "Rotation Angle (0-1)"=@n2) {materialFunction=&CustomRotator}.Rotated Values) {g=true}) {
   atlas=/Game/UI/ItemReceived/Materials/Textures/GiftingColorAtlas
   curve=/Game/UI/ItemReceived/Materials/Textures/GiftTag_OuterStroke_GradientCurve
   parameterName=CurveParams
  }
  n4 = VectorParameter() {parameterName=OutlineColor, defaultValue=<rgba 1 1 1 1>}
  n5 = ScalarParameter() {defaultValue=0.0156, parameterName=Threshold}
  n6 = Add(A=@n5, B=ScalarParameter() {defaultValue=0.163, parameterName=OuterStrokeSize})
  n7 = TextureSample() {texture=/Game/Global/Textures/Default/Blanks/T_Blank_White_ZeroAlpha_sRGBColor}
  n8 = ScalarParameter() {defaultValue=0.02, parameterName=Edge}
  n9 = ScalarParameter() {parameterName=SoftenEdge}
  n10 = MaterialFunctionCall!(Threshold=@n6, Gradient=@n7.R, Filter Width Override=@n8, Filter Width Fudge Factor=@n9) {materialFunction=/Game/UI/Materials/AAThreshold}
  n11 = TextureCoordinate()
  n12 = ScalarParameter() {parameterName=InnerColorDirection, sliderMax=1}
  n13 = CurveAtlasRowParameter(CurveTime=ComponentMask(None=MaterialFunctionCall(UVs=@n11, "Rotation Angle (0-1)"=@n12) {materialFunction=&CustomRotator}.Rotated Values) {g=true}) {
   atlas=&GiftingColorAtlas
   curve=/Game/UI/ItemReceived/Materials/Textures/GiftTag_InnerColor_GradientCurve
   defaultValue=1
   parameterName=CurveParams2
  }
  n14 = Add(A=@n6, B=ScalarParameter() {defaultValue=0.377, parameterName=OutlineSize})
  LinearInterpolate(
   A=LinearInterpolate(
    A=LinearInterpolate(A=@n3, B=@n4.RGB, Alpha=@n10.Result)
    B=@n13
    Alpha=MaterialFunctionCall(Threshold=@n14, Gradient=@n7.R, Filter Width Override=@n8, Filter Width Fudge Factor=@n9) {materialFunction=&AAThreshold}.Result
   )
   B=Constant3Vector() {constant=<rgba 1 1 1 0>}
   Alpha=ScalarParameter() {parameterName=WhiteFlood}
  )
  MaterialFunctionCall(Threshold=@n5, Gradient=@n7.R, Filter Width Override=@n8, Filter Width Fudge Factor=@n9) {materialFunction=&AAThreshold}
 ]
}
